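# Master masked, Nanite-ready material that erodes destroyed wood, stone or metal world pieces around a hit location.
# A noise-distorted sphere mask, scaled by radius, combines with world-aligned cracks and a channel-selected mask to cut opacity, while pushing vertices downward.
# Timestamp-driven timing, with debug overrides, feeds the animation; a Resource scalar selects color, metallic and roughness.
# Mobile and low quality use flat colors and cheaper noise-based cracks.
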Material /Game/Effects/Fort_Effects/World/Destruction/Destruction2/M_Destruction_Erosion_Master
graph {
 properties {
  bUseMaterialAttributes=true
  bUsedWithNanite=true
  bUsedWithSplineMeshes=true
  blendMode=BLEND_Masked
  opacityMaskClipValue=0.025
  pixelDepthOffsetMode=PDOM_Legacy
  translucencyLightingMode=TLM_VolumetricPerVertexDirectional
  usedShadingModels=Default Lit
 }
 nodes [
  n1 = ScalarParameter() {defaultValue=99, parameterName=Gravity Intensity}
  n2 = NamedRerouteUsage()
  n3 = AppendVector(A=Constant2Vector(), B=Multiply(A=LinearInterpolate(B=@n1, Alpha=@n2) {constB=3}) {constB=-1})
  Add(
   A=@n3
   B=Multiply(
    A=ScalarParameter() {defaultValue=33, group=DestructionDissolve, parameterName=Hit Vector Intensity, sortPriority=5}
    B=VectorParameter() {desc=Do Not Edit - Set in BP/Code, group=DestructionDissolve, parameterName=HitVector, sortPriority=4}.RGB
   )
  )
  Constant()
  Constant()
  n4 = VectorParameter() {parameterName=OpacityChannel, defaultValue=<rgba 1 0 0 0>}
  n5 = TextureSampleParameter2D() {parameterName=OpacityMask, texture=/Game/UI/GreyBox/white}
  n6 = Multiply(A=@n4.R, B=@n5.R)
  n7 = Multiply(A=@n4.G, B=@n5.G)
  n8 = Multiply(A=@n4.B, B=@n5.B)
  n9 = Multiply(A=@n4.A, B=@n5.A)
  n10 = TextureSample(UVs=TextureCoordinate() {uTiling=0.4, vTiling=0.4}) {samplerSource=SSM_Wrap_WorldGroupSettings, texture=/Game/TimeOfDay/LightFunction/T_Cloud2dNoise}
  n11 = ScalarParameter() {defaultValue=150, parameterName=intensity}
  n12 = ScalarParameter() {defaultValue=100, group=DestructionDissolve, parameterName=Sphere Radius, sortPriority=1}
  n13 = Divide(A=@n12) {constB=320}
  n14 = Max(A=Multiply(A=NamedRerouteDeclaration(None=@n13) {name=Compensation, nodeColor=<rgba 0.9688 1 0 1>}) {constB=3})
  n15 = NamedRerouteUsage()
  n16 = SphereMask(
   A=Add(A=Multiply(A=@n10.RGB, B=@n11) {constB=50}, B=WorldPosition().XYZ)
   B=VectorParameter() {parameterName="Hit Location (World Space)"}.RGB
   Radius=Multiply(A=Multiply(A=@n12, B=@n14) {constB=3}, B=Saturate(None=@n15))
   Hardness=ScalarParameter() {defaultValue=1, group=DestructionDissolve, parameterName=FadeHardness, sortPriority=2}
  )
  n17 = NamedRerouteUsage()
  n18 = Subtract(A=OneMinus(None=ComponentMask(None=@n17) {b=true}), B=ScalarParameter() {defaultValue=0.5, parameterName=Subtract}) {constB=0.75}
  n19 = LinearInterpolate(B=ScalarParameter() {defaultValue=16, parameterName=Crack INtensity}, Alpha=NamedRerouteUsage()) {constA=1}
  n20 = OneMinus(None=Saturate(None=Multiply(A=@n18, B=@n19) {constB=10}))
  n21 = LinearInterpolate(Alpha=NamedRerouteUsage()) {constA=1, constB=0}
  n22 = Multiply(A=Add(A=Add(A=@n6, B=@n7), B=Add(A=@n8, B=@n9)), B=Multiply(A=OneMinus(None=@n16), B=Multiply(A=@n20, B=@n21)))
  MakeMaterialAttributes!(
   BaseColor=Multiply(A=NamedRerouteUsage(), B=NamedRerouteUsage())
   Metallic=NamedRerouteUsage()
   Specular
   Roughness=NamedRerouteUsage()
   Anisotropy
   EmissiveColor
   Opacity=@n22
   OpacityMask=@n22
   Normal=NamedRerouteUsage()
   Tangent
   WorldPositionOffset=LinearInterpolate(B=@n3, Alpha=@n16)
   SubsurfaceColor
   ClearCoat
   ClearCoatRoughness
   AmbientOcclusion
   Refraction
   CustomizedUVs_0
   CustomizedUVs_1
   CustomizedUVs_2
   CustomizedUVs_3
   CustomizedUVs_4
   CustomizedUVs_5
   CustomizedUVs_6
   CustomizedUVs_7
   PixelDepthOffset
   ShadingModel
   Displacement
  )
  n23 = ScalarParameter() {defaultValue=1, parameterName=Resource}
  n24 = WorldPosition()
  n25 = ObjectPositionWS()
  n26 = ComponentMask(None=MaterialFunctionCall(
    TextureObject=TextureObject() {texture=&T_Cloud2dNoise}
    TextureSize=NamedRerouteUsage()
    WorldPosition=Subtract(A=@n24.XYZ, B=@n25)
   ) {materialFunction=&WorldAlignedTexture}.XYZ Texture) {r=true}
  n27 = LinearInterpolate(Alpha=@n26) {constA=0.125}
  n28 = Constant() {r=1}
  NamedRerouteDeclaration(None=If(A=@n23, "A > B"=@n27, "A == B"=@n27, "A < B"=@n28) {constB=3}) {name=Roughness, nodeColor=<rgba 0.9453 1 0 1>}
  n29 = ScalarParameter() {defaultValue=0.3, parameterName=Min Base Color Multiplier}
  n30 = ScalarParameter() {defaultValue=1, parameterName=Max Base Color Multiplier}
  n31 = ScalarParameter() {defaultValue=1, parameterName=Resource}
  n32 = Constant3Vector() {constant=<rgba 0.1 0.1 0.1 0>}
  n33 = NamedRerouteUsage()
  n34 = NamedRerouteUsage()
  n35 = If(A=@n31, "A > B"=If(A=@n31, "A > B"=@n32, "A == B"=@n33, "A < B"=@n33) {constB=3}, "A == B"=@n34, "A < B"=@n34) {constB=2}
  n36 = NamedRerouteUsage()
  NamedRerouteDeclaration(None=Multiply(A=LinearInterpolate(A=@n29, B=@n30, Alpha=@n26) {constA=0.1}, B=If(A=@n31, "A > B"=@n35, "A == B"=@n36, "A < B"=@n32) {constB=1})) {name=Base Color, nodeColor=<rgba 0 0.2734 1 1>}
  n37 = LinearInterpolate(
   A=VectorParameter() {parameterName=WoodColor_Light, defaultValue=<rgba 0.5417 0.3569 0.1726 1>}.RGB
   B=VectorParameter() {parameterName=WoodColor_Dark, defaultValue=<rgba 0.1354 0.0892 0.0432 1>}.RGB
   Alpha=TextureSample(UVs=TextureCoordinate() {uTiling=2, vTiling=2}) {samplerSource=SSM_Wrap_WorldGroupSettings, texture=&T_Cloud2dNoise}.B
  )
  n38 = VectorParameter() {parameterName=MobileWoodColor, defaultValue=<rgba 0.4 0.2679 0.1376 1>}
  NamedRerouteDeclaration(None=ShadingPathSwitch(Default=QualitySwitch(Default=@n37, Low=@n38.RGB), Mobile=@n38.RGB)) {name=Wood, nodeColor=<rgba 0.5938 0 1 1>}
  n39 = LinearInterpolate(
   A=VectorParameter() {parameterName=StoneColor_Light, defaultValue=<rgba 0.5423 0.5425 0.5371 1>}.RGB
   B=VectorParameter() {parameterName=StoneColor_Dark, defaultValue=<rgba 0.1356 0.1356 0.1343 1>}.RGB
   Alpha=TextureSample(UVs=TextureCoordinate() {uTiling=2, vTiling=2}) {samplerSource=SSM_Wrap_WorldGroupSettings, texture=&T_Cloud2dNoise}.B
  )
  n40 = VectorParameter() {parameterName=MobileStoneColor, defaultValue=<rgba 0.4 0.4 0.3918 1>}
  NamedRerouteDeclaration(None=ShadingPathSwitch(Default=QualitySwitch(Default=@n39, Low=@n40.RGB), Mobile=@n40.RGB)) {name=Stone, nodeColor=<rgba 0.5938 0 1 1>}
  n41 = LinearInterpolate(
   A=VectorParameter() {parameterName=MetalColor_Light, defaultValue=<rgba 0.5347 0.5347 0.5347 1>}.RGB
   B=VectorParameter() {parameterName=MetalColor_Dark, defaultValue=<rgba 0.07 0.07 0.07 1>}.RGB
   Alpha=TextureSample(UVs=TextureCoordinate() {uTiling=2, vTiling=2}) {samplerSource=SSM_Wrap_WorldGroupSettings, texture=&T_Cloud2dNoise}.B
  )
  n42 = VectorParameter() {parameterName=MobileMetalColor, defaultValue=<rgba 0.4 0.4 0.4 1>}
  NamedRerouteDeclaration(None=ShadingPathSwitch(Default=QualitySwitch(Default=@n41, Low=@n42.RGB), Mobile=@n42.RGB)) {name=Metal, nodeColor=<rgba 0.5938 0 1 1>}
  NamedRerouteDeclaration(None=If(A=@n23, "A > B"=@n28, "A == B"=@n28, "A < B"=Constant()) {constB=3}) {name=Metallic, nodeColor=<rgba 0 1 0.6953 1>}
  n43 = MaterialFunctionCall(
   TextureObject=TextureObject() {texture=/Game/Packages/jl_tests/CrackMaterial/T_crack_FX_MASK}
   TextureSize=NamedRerouteUsage()
   WorldPosition=VertexInterpolator(VS=PreSkinnedPosition()).PS
   World Space Normal=VertexInterpolator(VS=PreSkinnedNormal()).PS
  ) {materialFunction=&WorldAlignedTexture}
  n44 = TextureObject() {samplerType=SAMPLERTYPE_LinearColor, texture=/Game/Global/Textures/Default/Noise/T_Curie_NoisePack}
  n45 = NamedRerouteUsage()
  n46 = VertexInterpolator(VS=PreSkinnedPosition())
  n47 = StaticBool() {value=true}
  n48 = VertexInterpolator(VS=PreSkinnedNormal())
  n49 = Add(
   A=ComponentMask(None=MaterialFunctionCall(TextureObject=@n44, TextureSize=@n45, WorldPosition=@n46.PS, Export Float 4=@n47, World Space Normal=@n48.PS) {materialFunction=&WorldAlignedTexture}.XYZ Texture) {a=true}
   B=Constant() {r=0.16}
  ) {constB=0}
  NamedRerouteDeclaration(None=ShadingPathSwitch(Default=QualitySwitch(Default=@n43.XYZ Texture, Low=@n49), Mobile=@n49)) {name=Crack, nodeColor=<rgba 0.6875 0 1 1>}
  n50 = PreSkinnedNormal()
  n51 = PreSkinnedPosition()
  NamedRerouteDeclaration(None=MaterialFunctionCall!(
    TextureObject=TextureObject() {samplerType=SAMPLERTYPE_Normal, texture=/Game/Packages/jl_tests/CrackMaterial/T_crack_Normal}
    TextureSize=NamedRerouteUsage()
    WorldSpace
    WorldNormal=VertexInterpolator(VS=@n50).PS
    WorldPosition=VertexInterpolator(VS=@n51).PS
    Use High Quality Normals
    bEnableTextureSample
   ) {materialFunction=/Game/Packages/jl_tests/CrackMaterial/WorldAlignedNormal_Crack}.XYZ Texture) {name=Crack Normal, nodeColor=<rgba 0.125 0 1 1>}
  n52 = ScalarParameter() {defaultValue=0.0267, parameterName=DEBUG_TIME_CONTROL, sliderMax=1}
  n53 = Time() {bOverride_Period=true, period=0.225}
  n54 = ScalarParameter() {group=DestructionDissolve_Code, parameterName=DestructTimeStamp, sortPriority=0}
  n55 = ScalarParameter() {defaultValue=0.225, desc=Do NOT DELETE, group=DestructionDissolve_Code, parameterName=Opacity Time Length, sortPriority=0}
  n56 = Multiply(A=@n55, B=Clamp(None=NamedRerouteUsage()) {maxDefault=5, minDefault=0.5})
  n57 = StaticBool() {value=true}
  NamedRerouteDeclaration(None=StaticSwitchParameter(
    True=StaticSwitchParameter(True=@n52, False=@n53) {parameterName="DEBUG Manual Control "}
    False=MaterialFunctionCall(TimestampA=@n54, TransitionLength_In=@n56, Transition In Only?=@n57) {materialFunction=&MF_Timestamp_InOut_01}.SmoothStep
   ) {parameterName=DEBUG CONTROL}) {name=TimingControl, nodeColor=<rgba 0 0.1562 1 1>}
  NamedRerouteDeclaration(None=ScalarParameter() {defaultValue=700, group=Crack, parameterName=Texturesize}) {name=TextureSize, nodeColor=<rgba 1 0 0.375 1>}
  n58 = NamedRerouteUsage()
  NamedRerouteDeclaration(None=Saturate(None=Add(A=@n58) {constB=0.1})) {name=Crack Base Color Addition, nodeColor=<rgba 0 1 0.0391 1>}
 ]
}
