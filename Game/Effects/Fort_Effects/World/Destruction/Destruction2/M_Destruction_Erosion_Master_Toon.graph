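# Masked toon master material that erodes destructible wood, stone or metal objects outward from a world-space hit location.
# A timing control, normally timestamp-driven, feeds a noise-distorted sphere mask that cuts opacity and pushes vertices along Z as gravity.
# A second sphere mask minus the first forms a colored outline band, alongside world-aligned cracks from pre-skinned positions.
# A Resource parameter selects base color, metallic and roughness; debug switches can override the timing.
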Material /Game/Effects/Fort_Effects/World/Destruction/Destruction2/M_Destruction_Erosion_Master_Toon
graph {
 properties {
  bUseMaterialAttributes=true
  bUsedWithNanite=true
  bUsedWithSplineMeshes=true
  blendMode=BLEND_Masked
  opacityMaskClipValue=0.025
  pixelDepthOffsetMode=PDOM_Legacy
  translucencyLightingMode=TLM_VolumetricPerVertexDirectional
  usedShadingModels=Default Lit
 }
 nodes [
  n1 = VectorParameter() {parameterName=OpacityChannel, defaultValue=<rgba 1 0 0 0>}
  n2 = TextureSampleParameter2D() {parameterName=OpacityMask, texture=/Game/UI/GreyBox/white}
  n3 = Multiply(A=@n1.R, B=@n2.R)
  n4 = Multiply(A=@n1.G, B=@n2.G)
  n5 = Multiply(A=@n1.B, B=@n2.B)
  n6 = Multiply(A=@n1.A, B=@n2.A)
  n7 = TextureCoordinate()
  n8 = Add(
   A=Multiply(
    A=TextureSample(UVs=@n7) {samplerSource=SSM_Wrap_WorldGroupSettings, texture=/Game/Effects/Fort_Effects/GameplayPulginsAssets/Masamune/T_Noise04}.RGB
    B=ScalarParameter() {defaultValue=50, parameterName=intensity}
   ) {constB=50}
   B=WorldPosition().XYZ
  )
  n9 = VectorParameter() {parameterName="Hit Location (World Space)"}
  n10 = Saturate(None=NamedRerouteUsage())
  n11 = ObjectRadius()
  n12 = Add(A=@n11, B=Multiply(A=ScalarParameter() {defaultValue=10, parameterName=Outline Thickness}) {constB=-1}) {constB=-2}
  n13 = ScalarParameter() {defaultValue=2.5, parameterName=Multiply}
  n14 = Max(A=Multiply(A=Divide(A=@n12, B=@n12) {constB=320}, B=@n13) {constB=2})
  n15 = ScalarParameter() {defaultValue=1, group=DestructionDissolve, parameterName=FadeHardness, sortPriority=2}
  n16 = SphereMask(A=@n8, B=@n9.RGB, Radius=Multiply(A=@n10, B=Multiply(A=@n12, B=@n14) {constB=3}), Hardness=@n15)
  n17 = Multiply(A=Add(A=Add(A=@n3, B=@n4), B=Add(A=@n5, B=@n6)), B=OneMinus(None=@n16))
  n18 = Constant2Vector()
  n19 = Multiply(A=LinearInterpolate(B=ScalarParameter() {defaultValue=10, parameterName=Gravity Intensity}, Alpha=NamedRerouteUsage()) {constB=3}) {constB=-1}
  MakeMaterialAttributes!(
   BaseColor=NamedRerouteUsage()
   Metallic=NamedRerouteUsage()
   Specular
   Roughness=NamedRerouteUsage()
   Anisotropy
   EmissiveColor
   Opacity=@n17
   OpacityMask=@n17
   Normal
   Tangent
   WorldPositionOffset=LinearInterpolate(B=AppendVector(A=@n18, B=@n19), Alpha=@n16)
   SubsurfaceColor
   ClearCoat
   ClearCoatRoughness
   AmbientOcclusion
   Refraction
   CustomizedUVs_0
   CustomizedUVs_1
   CustomizedUVs_2
   CustomizedUVs_3
   CustomizedUVs_4
   CustomizedUVs_5
   CustomizedUVs_6
   CustomizedUVs_7
   PixelDepthOffset
   ShadingModel
   Displacement
  )
  Multiply(A=ObjectRadius(), B=ScalarParameter() {defaultValue=2.5, parameterName=Multiply})
  n20 = OneMinus(None=ComponentMask(None=NamedRerouteUsage()) {g=true})
  n21 = ScalarParameter() {parameterName=Subtract}
  n22 = ScalarParameter() {defaultValue=16, parameterName=Crack INtensity}
  n23 = NamedRerouteUsage()
  n24 = Saturate(None=Multiply(A=Subtract(A=@n20, B=@n21) {constB=0.75}, B=LinearInterpolate(B=@n22, Alpha=@n23) {constA=1}) {constB=10})
  n25 = NamedRerouteUsage()
  Multiply(B=Multiply(A=OneMinus(None=@n24), B=LinearInterpolate(Alpha=@n25) {constA=1, constB=0}))
  n26 = ScalarParameter() {defaultValue=1, parameterName=Resource}
  n27 = Constant() {r=1}
  n28 = ScalarParameter() {defaultValue=0.125, parameterName=Metallic Roughness}
  NamedRerouteDeclaration(None=LinearInterpolate(A=If(A=@n26, "A > B"=@n27, "A == B"=@n28, "A < B"=@n27) {constB=3}, Alpha=NamedRerouteUsage()) {constB=0}) {name=Roughness, nodeColor=<rgba 0.9453 1 0 1>}
  n29 = ScalarParameter() {defaultValue=1, parameterName=Resource}
  n30 = Constant3Vector() {desc=Fallback / Default, constant=<rgba 0.05 0.05 0.05 0>}
  n31 = NamedRerouteUsage()
  n32 = NamedRerouteUsage()
  n33 = If(A=@n29, "A > B"=If(A=@n29, "A > B"=@n30, "A == B"=@n31, "A < B"=@n31) {constB=3}, "A == B"=@n32, "A < B"=@n32) {constB=2}
  n34 = NamedRerouteUsage()
  n35 = NamedRerouteUsage()
  n36 = Add(A=NamedRerouteUsage(), B=OneMinus(None=ComponentMask(None=@n35) {g=true}))
  NamedRerouteDeclaration(None=LinearInterpolate(
    A=If(A=@n29, "A > B"=@n33, "A == B"=@n34, "A < B"=@n30) {constB=1}
    B=VectorParameter() {parameterName=Outline Color}.RGB
    Alpha=Saturate(None=@n36)
   ) {constB=0}) {name=Base Color, nodeColor=<rgba 0 0.2734 1 1>}
  NamedRerouteDeclaration(None=VectorParameter() {parameterName=MobileWoodColor, defaultValue=<rgba 0.2 0.134 0.0688 1>}.RGB) {name=Wood, nodeColor=<rgba 0.5938 0 1 1>}
  NamedRerouteDeclaration(None=VectorParameter() {parameterName=MobileStoneColor, defaultValue=<rgba 0.2 0.2 0.1959 1>}.RGB) {name=Stone, nodeColor=<rgba 0.5938 0 1 1>}
  NamedRerouteDeclaration(None=VectorParameter() {parameterName=MobileMetalColor, defaultValue=<rgba 0.1 0.1 0.1 1>}.RGB) {name=Metal, nodeColor=<rgba 0.5938 0 1 1>}
  n37 = Constant()
  NamedRerouteDeclaration(None=If(A=@n26, "A > B"=@n37, "A == B"=@n27, "A < B"=@n37) {constB=3}) {name=Metallic, nodeColor=<rgba 0 1 0.6953 1>}
  n38 = MaterialFunctionCall(
   TextureObject=TextureObject() {texture=&/T_Toon_DamageCracks}
   TextureSize=NamedRerouteUsage()
   WorldPosition=VertexInterpolator(VS=PreSkinnedPosition()).PS
   World Space Normal=VertexInterpolator(VS=PreSkinnedNormal()).PS
  ) {materialFunction=&WorldAlignedTexture}
  n39 = TextureObject() {texture=&T_Toon_DamageCracks}
  n40 = NamedRerouteUsage()
  n41 = VertexInterpolator(VS=PreSkinnedPosition())
  n42 = StaticBool() {value=true}
  n43 = VertexInterpolator(VS=PreSkinnedNormal())
  n44 = Add(
   A=ComponentMask(None=MaterialFunctionCall(TextureObject=@n39, TextureSize=@n40, WorldPosition=@n41.PS, Export Float 4=@n42, World Space Normal=@n43.PS) {materialFunction=&WorldAlignedTexture}.XYZ Texture) {g=true}
   B=Constant() {r=0.16}
  ) {constB=0}
  NamedRerouteDeclaration(None=ShadingPathSwitch(Default=QualitySwitch(Default=@n38.XYZ Texture, Low=@n44), Mobile=@n44)) {name=Crack, nodeColor=<rgba 0.6875 0 1 1>}
  n45 = ScalarParameter() {defaultValue=0.5, parameterName=DEBUG_TIME_CONTROL, sliderMax=1}
  n46 = Time() {bOverride_Period=true, period=0.225}
  n47 = ScalarParameter() {group=DestructionDissolve_Code, parameterName=DestructTimeStamp, sortPriority=0}
  n48 = ScalarParameter() {defaultValue=0.225, desc=Do NOT DELETE, group=DestructionDissolve_Code, parameterName=Opacity Time Length, sortPriority=0}
  n49 = Multiply(A=@n48, B=Clamp(None=NamedRerouteUsage()) {maxDefault=5, minDefault=0.5})
  n50 = StaticBool() {value=true}
  NamedRerouteDeclaration(None=StaticSwitchParameter(
    True=StaticSwitchParameter(True=@n45, False=@n46) {parameterName="DEBUG Manual Control "}
    False=MaterialFunctionCall(TimestampA=@n47, TransitionLength_In=@n49, Transition In Only?=@n50) {materialFunction=&MF_Timestamp_InOut_01}.SmoothStep
   ) {parameterName=DEBUG CONTROL}) {name=TimingControl, nodeColor=<rgba 0 0.1562 1 1>}
  NamedRerouteDeclaration(None=ScalarParameter() {defaultValue=364, group=Crack, parameterName=Texturesize}) {name=TextureSize, nodeColor=<rgba 1 0 0.375 1>}
  n51 = Multiply(A=NamedRerouteDeclaration(None=Divide(A=@n11, B=@n11) {constB=320}) {name=Compensation, nodeColor=<rgba 0.9688 1 0 1>}, B=@n13) {constB=2}
  n52 = Multiply(A=Multiply(A=@n11, B=Max(A=@n51)) {constB=3}, B=@n10)
  NamedRerouteDeclaration(None=Subtract(A=SphereMask(A=@n8, B=@n9.RGB, Radius=@n52, Hardness=@n15), B=@n16)) {name=Outline, nodeColor=<rgba 0.2656 1 0 1>}
  TextureSample() {samplerSource=SSM_Wrap_WorldGroupSettings, texture=/Game/TimeOfDay/LightFunction/T_Cloud2dNoise}
 ]
}
