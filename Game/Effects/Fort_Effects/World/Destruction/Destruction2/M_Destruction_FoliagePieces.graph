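# Masked, two-sided subsurface material for foliage debris pieces on Niagara mesh particles in destruction effects.
# A dynamic parameter selects the atlas sub-image, blends base color toward a burnt look and scales glow.
# Particle color tints diffuse and subsurface from packed textures, while sphere normals blend with detail normals.
# Glow combines Fresnel with a fire curve atlas and boosts HDR outside forward and mobile paths, and a time-driven sine provides wobble offset.
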Material /Game/Effects/Fort_Effects/World/Destruction/Destruction2/M_Destruction_FoliagePieces
graph {
 properties {
  bUseMaterialAttributes=true
  bUsedWithNiagaraMeshParticles=true
  blendMode=BLEND_Masked
  pixelDepthOffsetMode=PDOM_Legacy
  shadingModel=MSM_Subsurface
  twoSided=true
  usedShadingModels=Subsurface
 }
 nodes [
  n1 = NamedRerouteUsage()
  n2 = Desaturation(None=NamedRerouteUsage(), Fraction=Constant() {r=0.5})
  n3 = TextureSample(UVs=NamedRerouteUsage()) {texture=&/T_FoliageDestruction_ColorAlphaMask}
  MakeMaterialAttributes!(
   BaseColor=LinearInterpolate(
    A=NamedRerouteUsage()
    B=Multiply(A=@n1, B=@n2)
    Alpha=DynamicParameter() {defaultValue=<rgba 0 0 1 0>, paramNames [Glow, Burnt, ----, SubUVSelection]}.Burnt
   )
   Metallic=Constant()
   Specular=Constant() {r=0.5}
   Roughness=NamedRerouteUsage()
   Anisotropy
   EmissiveColor=NamedRerouteUsage()
   Opacity=@n3.G
   OpacityMask=@n3.G
   Normal=NamedRerouteUsage()
   Tangent
   WorldPositionOffset=NamedRerouteUsage()
   SubsurfaceColor=NamedRerouteUsage()
   ClearCoat
   ClearCoatRoughness
   AmbientOcclusion
   Refraction
   CustomizedUVs_0
   CustomizedUVs_1
   CustomizedUVs_2
   CustomizedUVs_3
   CustomizedUVs_4
   CustomizedUVs_5
   CustomizedUVs_6
   CustomizedUVs_7
   PixelDepthOffset
   ShadingModel
   Displacement
  )
  n4 = TextureCoordinate()
  n5 = AppendVector(A=ScalarParameter() {defaultValue=4, parameterName=U_Images}, B=ScalarParameter() {defaultValue=4, parameterName=V_Images})
  n6 = ScalarParameter() {defaultValue=4, parameterName=V_Images}
  n7 = Multiply(A=DynamicParameter() {defaultValue=<rgba 0 0 1 0>, paramNames [Glow, Burnt, ----, SubUVSelection]}.SubUVSelection, B=@n6)
  n8 = ScalarParameter() {defaultValue=4, parameterName=U_Images}
  n9 = Multiply(A=Frac(None=@n7), B=@n8)
  n10 = Multiply(A=Frac(None=Divide(A=@n7, B=@n6)), B=@n6)
  n11 = AppendVector(A=Divide(A=Floor(None=@n9), B=@n8), B=Divide(A=Floor(None=@n10), B=@n6))
  NamedRerouteDeclaration(None=Add(A=Divide(A=@n4, B=@n5), B=@n11)) {name=SubUVControl_Opacity, nodeColor=<rgba 0 0.4375 1 1>}
  n12 = TextureCoordinate() {uTiling=4, vTiling=4}
  n13 = AppendVector(A=ScalarParameter() {defaultValue=4, parameterName=U_Images}, B=ScalarParameter() {defaultValue=4, parameterName=V_Images})
  NamedRerouteDeclaration(None=Add(A=Divide(A=@n12, B=@n13), B=@n11)) {name=SubUVControl_Textures, nodeColor=<rgba 0 0.3672 1 1>}
  n14 = TextureSample(UVs=NamedRerouteUsage()) {
   samplerType=SAMPLERTYPE_LinearGrayscale
   texture=/Game/Characters/Player/Male/Medium/Bodies/M_Med_Soldier_04/Skins/Gingerbread_Burnt/Textures/BurnGradient_output
  }
  n15 = NamedRerouteUsage()
  NamedRerouteDeclaration(None=Multiply(
    A=LinearInterpolate(Alpha=@n14.G) {constA=0.05, constB=0.9}
    B=TextureSample(UVs=@n15) {texture=&T_FoliageDestruction_ColorAlphaMask}.G
   )) {name=Burnt Sample, nodeColor=<rgba 0.2422 1 0 1>}
  n16 = ParticleColor()
  n17 = NamedRerouteUsage()
  NamedRerouteDeclaration(None=Multiply(
    A=@n16.RGB
    B=TextureSample(UVs=@n17) {texture=&T_FoliageDestruction_ColorAlphaMask}.R
   )) {name=Diffuse, nodeColor=<rgba 0 0.1328 1 1>}
  n18 = TextureSample(UVs=@n17) {texture=&/T_FoliageDestruction_SSHR}
  NamedRerouteDeclaration(None=@n18.B) {name=Roughness, nodeColor=<rgba 1 0 0.6328 1>}
  NamedRerouteDeclaration(None=Multiply(A=@n16.RGB, B=@n18.G)) {name=SubSurface, nodeColor=<rgba 0.9688 0 1 1>}
  n19 = MaterialFunctionCall(
   BaseNormal=TextureSample(UVs=@n17) {samplerType=SAMPLERTYPE_LinearColor, texture=/Game/Packages/Fortress_SharedMaterials/Textures/SphereNormalsUncomp}.RGB
   AdditionalNormal=TextureSample(UVs=@n17) {samplerType=SAMPLERTYPE_Normal, texture=&/T_FoliageDestruction_N}.RGB
  ) {materialFunction=&BlendAngleCorrectedNormals}
  NamedRerouteDeclaration(None=@n19.Result) {name=Normal, nodeColor=<rgba 1 0 0.1406 1>}
  NamedRerouteDeclaration(None=ComponentMask(None=@n19.Result) {b=true}) {name=Glow Normal Read, nodeColor=<rgba 1 0.9844 0 1>}
  n20 = Multiply(A=NamedRerouteUsage(), B=NamedRerouteUsage())
  n21 = ScalarParameter() {defaultValue=1, group=Glow, parameterName=FresnelExp, sortPriority=2}
  n22 = Saturate(None=Add(A=LinearInterpolate(Alpha=@n20), B=Fresnel(ExponentIn=@n21) {baseReflectFraction=0, exponent=1}))
  n23 = CurveAtlasRowParameter(CurveTime=@n22) {
   atlas=/Game/Curie/Effects/temp/Curves/CA_Curie_Fire_Visuals
   curve=/Game/Curie/Effects/temp/Curves/Curve_Fiery
   defaultValue=2
   group=Glow
   parameterName=Curie
  }
  n24 = Multiply(A=@n23, B=Add(A=@n22) {constB=0.1})
  n25 = Saturate(None=@n24)
  n26 = ScalarParameter() {defaultValue=1, group=Glow, parameterName=HDRIntensity, sortPriority=3}
  n27 = Multiply(
   A=ShadingPathSwitch(
    Default=MaterialFunctionCall(LDR Input=@n25, HDR Tint=@n24, HDR Intensity Multiplier=@n26) {materialFunction=&DeriveHDRfromLDR}.RemappedRange
    Forward=@n25
    Mobile=@n25
   )
   B=DynamicParameter() {defaultValue=<rgba 0 0 1 0>, paramNames [Glow, Burnt, ----, SubUVSelection]}.Glow
  )
  n28 = Multiply(A=NamedRerouteUsage()) {constB=0.1}
  NamedRerouteDeclaration(None=MaterialFunctionCall(Emissive=Add(A=@n27, B=@n28)) {materialFunction=&MF_FX_EmissiveAdjustments}.Result) {name=Emissive, nodeColor=<rgba 1 0 0.1875 1>}
  n29 = Add(
   A=Multiply(A=Time(), B=ScalarParameter() {defaultValue=0.5, parameterName=WPO_Speed})
   B=Divide(A=TextureCoordinate(), B=ScalarParameter() {defaultValue=0.5, parameterName="WaveScale (S)"})
  )
  NamedRerouteDeclaration(None=Multiply(A=Sine(None=@n29) {period=2}, B=ScalarParameter() {defaultValue=3, parameterName="SmallWaveScle (S)"})) {name=WPO, nodeColor=<rgba 0.8047 1 0 1>}
 ]
}
