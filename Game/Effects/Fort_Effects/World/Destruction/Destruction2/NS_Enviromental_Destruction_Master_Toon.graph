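# Toon-styled master Niagara system for environment destruction, bursting debris chosen by User.Resource (wood, stone, metal, foliage, glass).
# GPU Resource_Query snaps mesh chunks to nearby rigid-body surfaces and feeds Metal_Sparks; CPU mobile variants use simpler box spawning.
# Hit location, vector and bounds shape spawning and impulses, with counts scaled by object radius.
# Glass adds shards and a Sound Library event; smoke and toon lines skip glass and instant-death hits.
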NiagaraSystem /Game/Effects/Fort_Effects/World/Destruction/Destruction2/NS_Enviromental_Destruction_Master_Toon
dependencies {
 usedDataInterfaces [
  &NiagaraDataInterfaceParticleRead
  &NiagaraDataInterfaceCurve
  &NiagaraDataInterfaceVectorCurve
  /Script/SoundLibrary.NiagaraDataInterfaceSoundLibraryPlayer
  &NiagaraDataInterfaceArrayFloat3
  &NiagaraDataInterfaceMeshRendererInfo
  &NiagaraDataInterfaceArrayFloat
  &NiagaraDataInterfaceGBuffer
  /Script/Niagara.NiagaraDataInterfaceRigidMeshCollisionQuery
 ]
 usedDynamicInputs [
  &SimulationPosition
  &SelectFloatFromArray
  &self:ScaleFactor
  &ClampFloat
  &Divide_Float
 ]
 usedModules [
  &self:SystemSpawnScript.SystemScriptSource.SystemScriptGraph.NiagaraNodeAssignment_1.SetVariables_#1
  &SystemState
  &self:SystemSpawnScript.SystemScriptSource.SystemScriptGraph.NiagaraNodeAssignment_0.SetVariables_#2
  &self:SystemSpawnScript.SystemScriptSource.SystemScriptGraph.NiagaraNodeAssignment_2.SetVariables_#3
 ]
}
emitters {
 Audio_Collision {
  bEnabled=false
  emitterUpdateScript [
   &EmitterState()
   &SpawnBurst_Instantaneous(Use Spawn Probability=false, UseLoopCountLimit=true)
  ]
  particleSpawnScript [
   &InitializeParticle(
    Lifetime=2
    Mass Max=1
    Mass Min=0.5
    Mass Mode=::NewEnumerator3
    Mesh Scale Mode=::NewEnumerator3
    Sprite Size Mode=::NewEnumerator3
    Uniform Sprite Size=20
    UsePositionOffset=false
    Write Material Random=true
   )
   &self:ResourceSetup_AudioCollision(
    Default Mass 0:NiagaraFloat=&RandomRangeFloat(Minimum=2, Maximum=7)
    Default Mass 1:NiagaraFloat=&RandomRangeFloat(Minimum=3, Maximum=9)
    Default Mass 2:NiagaraFloat=&RandomRangeFloat(Minimum=5, Maximum=17)
    Default Scale 0:Vector3f=&RandomRangeVector(Minimum=<0.5 0.5 0.5>, Maximum=<1 1 1>)
    Default Scale 1:Vector3f=&RandomRangeVector(Minimum=<1 1 1>, Maximum=<3 3 3>)
    Default Scale 2:Vector3f=&RandomRangeVector(Minimum=<4 4 4>, Maximum=<6 6 6>)
    Foliage Mass:NiagaraFloat=&RandomRangeFloat(Minimum=0.5, Maximum=1.5)
    Foliage Scale:Vector3f=&RandomRangeVector(Minimum=<5 5 5>, Maximum=<12 12 12>)
    Glass Mass 0:NiagaraFloat=&RandomRangeFloat(Minimum=0.75, Maximum=1.75)
    Glass Mass 1:NiagaraFloat=&RandomRangeFloat(Minimum=1.5, Maximum=5)
    Glass Mass 2:NiagaraFloat=&RandomRangeFloat(Minimum=3.5, Maximum=10)
    Glass Scale 0:Vector3f=&RandomRangeVector(Minimum=<0.5 0.5 0.5>, Maximum=<1.5 1.5 1.5>)
    Glass Scale 1:Vector3f=&RandomRangeVector(Minimum=<0.5 0.5 0.5>, Maximum=<2 2 2>)
    Glass Scale 2:Vector3f=&RandomRangeVector(Minimum=<4 4 4>, Maximum=<6 6 6>)
    Medium Mesh=0.96
    Metal Mass 0:NiagaraFloat=&RandomRangeFloat(Minimum=2.5, Maximum=7)
    Metal Mass 1:NiagaraFloat=&RandomRangeFloat(Minimum=3.5, Maximum=9)
    Metal Mass 2:NiagaraFloat=&RandomRangeFloat(Minimum=5.5, Maximum=17)
    Metal Scale 0:Vector3f=&RandomRangeVector(Minimum=<0.5 0.5 0.5>, Maximum=<1 1 1>)
    Metal Scale 1:Vector3f=&RandomRangeVector(Minimum=<0.1 0.1 0.1>, Maximum=<0.5 0.5 0.5>)
    Metal Scale 2:Vector3f=&RandomRangeVector(Minimum=<0.3 0.3 0.3>, Maximum=<1 1 1>)
    Module.Override Randomness=false
    PhysicalSurface:NiagaraInt32=$User.Resource
    Small Mesh=0.5
    Stone Mass 0:NiagaraFloat=&RandomRangeFloat(Minimum=2, Maximum=7)
    Stone Mass 1:NiagaraFloat=&RandomRangeFloat(Minimum=3, Maximum=9)
    Stone Mass 2:NiagaraFloat=&RandomRangeFloat(Minimum=5, Maximum=17)
    Stone Scale 0:Vector3f=&RandomRangeVector(Minimum=<0.3 0.3 0.3>, Maximum=<1 1 1>)
    Stone Scale 1:Vector3f=&RandomRangeVector(Minimum=<0.5 0.5 0.5>, Maximum=<2 2 2>)
    Stone Scale 2:Vector3f=&RandomRangeVector(Minimum=<4 4 4>, Maximum=<6 6 6>)
    Wood Mass 0:NiagaraFloat=&RandomRangeFloat(Minimum=1.5, Maximum=8)
    Wood Mass 1:NiagaraFloat=&RandomRangeFloat(Minimum=2.5, Maximum=9)
    Wood Mass 2:NiagaraFloat=&RandomRangeFloat(Minimum=4.5, Maximum=17)
    Wood Scale 0:Vector3f=&RandomRangeVector(Minimum=<0.8 0.8 0.8>, Maximum=<1.2 1.2 1.2>)
    Wood Scale 1:Vector3f=&RandomRangeVector(Minimum=<2 2 2>, Maximum=<3 3 3>)
    Wood Scale 2:Vector3f=&RandomRangeVector(Minimum=<0.75 0.75 0.75>, Maximum=<1 1 1>)
   ) {module=ResourceSetup_AudioCollision}
   /SetVariables(Particles.SpriteSize=&Multiply_Vector2D(
     A=$Particles.SpriteSize
     B=&MakeVector2D(
      X=&MakeFloatFromVector(Vector=$Particles.Scale, Channel=::NewEnumerator0)
      Y=&MakeFloatFromVector(Vector=$Particles.Scale, Channel=::NewEnumerator1)
     )
    )) {isSetParameters=true}
   &PointForce(
    Force Falloff Distance:NiagaraFloat=$User.Radius
    Force Origin:NiagaraPosition=$User.HitLocation
    Force Strength=&MakeCustomFloatFromBool(
     Bool=$User.bIsInstantDeath
     True Float=&Multiply_Float(A=&RandomRangeFloat(Minimum=1000, Maximum=1500), B=$System.ImpulseScale)
     False Float=&Multiply_Float(A=&RandomRangeFloat(Minimum=1750, Maximum=2250), B=$System.ImpulseScale)
    )
    Module.Use Falloff=true
    Module.UseLinearForce=false
   )
   &LinearForce(
    Coordinate Space:ENiagaraCoordinateSpace=::World
    Force=&Multiply_VectorByFloat(
     Vector=$User.HitVector
     Float=&MakeCustomFloatFromBool(
      Bool=$User.bIsInstantDeath
      True Float=&Multiply_Float(A=&RandomRangeFloat(Minimum=1250, Maximum=1750), B=$System.ImpulseScale)
      False Float=&Multiply_Float(A=&RandomRangeFloat(Minimum=1500, Maximum=2000), B=$System.ImpulseScale)
     )
    )
   )
   &ApplyInitialForces()
  ]
  particleUpdateScript [
   &ParticleState()
   &GravityForce()
   &Drag(
    Drag=&self:Gravity_Drag_Compensation(Drag Falloff:NiagaraFloat=2, Gravity Direction:Vector3f=<0 0 -1>)
    Use Linear Drag=true
    Use Rotational Drag=false
   )
   &Collision(
    Age Colliding Particles=false
    Friction During a Bounce=0.2
    Module.Control Rest State Via Surface Normal=false
    Module.EnableMaxCollisionCount=false
    Module.Randomize Collision Normal Vector=true
    Restitution=0.15
   )
   /Niagara/Modules/Events/GenerateCollisionEvent(Module.EventDelay=true, Module.Use Velocity Threshold=true)
   &SolveForcesAndVelocity(Clamp Velocity=false, Limit Acceleration=false)
  ]
  rendererClasses []
  renderers []
 }
 Glass {
  simTarget=GPUComputeSim
  emitterUpdateScript [
   &EmitterState(Life Cycle Mode=::NewEnumerator1, Loop Behavior=::NewEnumerator1, Loop Duration=0.1, UseLoopDelay=false)
   &SpawnBurst_Instantaneous(
    Spawn Count=&MakeCustomIntFromBool(
     Bool=&SetBoolByIntComparison(A=$User.Resource, B=5, Comparison Type=::NewEnumerator4)
     True INT=&MakeIntFromFloat(Float=&Multiply_Float(A=175, B=$System.ObjectRadiusScaleFactor))
     False INT=0
    )
    Use Spawn Probability=false
    UseLoopCountLimit=false
   )
  ]
  particleSpawnScript [
   /SetVariables(Particles.glassMeshSelection:NiagaraInt32=&RandomRangeInt(Minimum=0, Maximum=2)) {isSetParameters=true}
   &InitializeParticle(
    Color Mode=::NewEnumerator0
    Lifetime Mode=::NewEnumerator1
    Mesh Renderer Array Visibility Mode=::NewEnumerator1
    Mesh Renderer Info=DataInterface() {propertyValues {MeshRenderer=&self:Glass_0.NiagaraMeshRendererProperties_5}}
    Mesh Renderer Mesh Index:NiagaraInt32=$Particles.glassMeshSelection
    Mesh Scale Max=&SelectVectorFromArray(
     Direct Array Index=$Particles.glassMeshSelection
     Vector Selection Array=DataInterface() {propertyValues {FloatData [<1.25 1.25 1.25>, <2 2 2>, <2.75 2.75 2.75>]}}
    )
    Mesh Scale Min=&SelectVectorFromArray(
     Direct Array Index=$Particles.glassMeshSelection
     Vector Selection Array=DataInterface() {propertyValues {FloatData [<0.25 0.25 0.25>, <0.75 0.75 0.75>, <1 1 1>]}}
    )
    Mesh Scale Mode=::NewEnumerator2
    UsePositionOffset=false
    Write Material Random=true
   )
   &ShapeLocation(
    Box Size=&Multiply_Vector(A=&Multiply_VectorByFloat(Vector=$User.Bounds, Float=2), B=$User.Scale)
    MaskSpawnGroup=false
    Override Seed=false
    Shape Primitive=::NewEnumerator2
    Surface Only=false
    UseOwnerScale=true
   )
   &InitialMeshOrientation(
    Mesh Axis To Orient=<0 0 1>
    Mesh Orientation Mode=::NewEnumerator2
    Orientation Coordinate Space=::World
    Orientation Vector=$Particles.ShapeLocation.ShapeNormal
    Rotation=&RandomRangeVector(Minimum=<0 0 0>, Maximum=<0 0 1>)
    Use Rotation Vector=true
   )
   &LinearForce(
    Coordinate Space:ENiagaraCoordinateSpace=::World
    Force=&Multiply_VectorByFloat(
     Vector=&Multiply_Vector(A=$User.HitVector, B=<1 1 0.5>)
     Float=&Multiply_Float(A=&RandomRangeFloat(Minimum=750, Maximum=1500), B=$System.ImpulseScale)
    )
   )
   &PointForce(
    Force Origin:NiagaraPosition=$User.HitLocation
    Force Strength=&MakeCustomFloatFromBool(
     Bool=$User.bIsInstantDeath
     True Float=&Multiply_Float(A=&RandomRangeFloat(Minimum=1000, Maximum=1500), B=$System.ImpulseScale)
     False Float=&Multiply_Float(A=&RandomRangeFloat(Minimum=250, Maximum=500), B=$System.ImpulseScale)
    )
    Module.Use Falloff=false
    Module.UseLinearForce=false
   )
   &ApplyInitialForces()
  ]
  particleUpdateScript [
   &ParticleState()
   &Collision(
    Age Colliding Particles=false
    Collision Enabled=&RandomBool()
    Friction During a Bounce=0.2
    GPU Collision Type=::NewEnumerator2
    Mesh Dimensions:Vector3f=&SelectVectorFromArray(
     Direct Array Index=$Particles.glassMeshSelection
     Vector Selection Array=DataInterface() {propertyValues {FloatData [<12 5 1>, <16 11 1>, <22 13 2>]}}
    )
    Module.Control Rest State Via Surface Normal=false
    Module.EnableMaxCollisionCount=false
    Module.Randomize Collision Normal Vector=true
    Particle Radius Scale=0.1
    Radius Calculation Type=::NewEnumerator1
    Restitution=0.15
   )
   &UpdateMeshOrientation(Rotation Vector=&MakeCustomVectorFromBool(
     Bool=$Particles.HasCollided
     True Vector=<0 0 0>
     False Vector=&RandomRangeVector(Minimum=<-1 -1 -1>, Maximum=<1 1 1>)
    ))
   &Drag(
    Drag=&self:Gravity_Drag_Compensation(Drag Falloff:NiagaraFloat=2, Gravity Direction:Vector3f=<0 0 -1>)
    Use Linear Drag=true
    Use Rotational Drag=true
   )
   &GravityForce(Gravity=&Multiply_VectorByFloat(
     Vector=<0 0 -980>
     Float=&FloatFromCurve(FloatCurve=DataInterface() {propertyValues {
        bOptimizeLUT=false
        Curve {keys [<at 0.05 0 RCIM_Cubic>, {arriveTangent=45.9003, interpMode=RCIM_Cubic, leaveTangent=45.9003, tangentMode=RCTM_User, time=0.1, value=1}]}
       }})
    ))
   &SolveForcesAndVelocity(Clamp Velocity=true, Limit Acceleration=true, Speed Limit=2500)
   &ScaleMeshSize(Scale Factor=&VectorFromFloat(Value=&FloatFromCurve(FloatCurve=DataInterface() {propertyValues {Curve {keys [
          <at 0 0 RCIM_Constant>
          <at 0.05 1 RCIM_Constant>
          {interpMode=RCIM_Cubic, leaveTangentWeight=0.3333, tangentMode=RCTM_User, time=0.75, value=1}
          {arriveTangentWeight=0.3333, interpMode=RCIM_Cubic, tangentMode=RCTM_User, time=1, value=0}
         ]}}})))
  ]
  rendererClasses [&NiagaraMeshRendererProperties]
  renderers [{rendererClass=&NiagaraMeshRendererProperties, rendererIndex=0}]
 }
 Metal_Sparks {
  simTarget=GPUComputeSim
  emitterUpdateScript [
   &EmitterState(Life Cycle Mode=::NewEnumerator1, Loop Behavior=::NewEnumerator1, Loop Duration=0.25, UseLoopDelay=false)
   &SpawnParticlesFromOtherEmitter(
    Attribute Reader=DataInterface() {propertyValues {EmitterBinding {bindingMode=Other, emitterName=Resource_Query}}}
    Spawn Rate Per Particle Cap:NiagaraFloat=&MakeCustomFloatFromBool(Bool=$System.isMetal, True Float=100, False Float=0)
    Use Maximum Age=false
    Use Minimum Age=false
   )
  ]
  particleSpawnScript [
   &InitializeParticle(
    Color=<rgba 30 26.9412 16 1>
    Lifetime=&UniformRangedFloat(Minimum=0.1, Maximum=0.45)
    Mass Mode=::NewEnumerator1
    Sprite Size=&UniformRangedVector2D(Minimum=<1 1>, Maximum=<2 2>)
    Sprite Size Mode=::NewEnumerator4
    UsePositionOffset=false
    Write Material Random=true
   ) {module=InitializeParticle001}
   &ShapeLocation(MaskSpawnGroup=false, Override Seed=false, Sphere Radius=5, UseOwnerScale=true)
   &SampleParticlesFromOtherEmitter(Velocity Sampling=::NewEnumerator0)
   &CameraOffset(Camera Offset Amount=25)
   &AddVelocity(Module.Use Falloff Distance=false, Override Seed=false, Velocity Mode=::NewEnumerator1, Velocity Speed=500)
  ]
  particleUpdateScript [
   &ParticleState()
   &AccelerationForce(Acceleration=<0 0 -980>)
   &Drag(Drag=&RandomRangeFloat(Minimum=1, Maximum=3), Use Linear Drag=true, Use Rotational Drag=false)
   &SolveForcesAndVelocity(Clamp Velocity=false, Limit Acceleration=false)
   &ScaleSpriteSizeBySpeed(Max Scale Factor=<1 10>)
   &ScaleSpriteSize(Scale Factor=&Vector2DFromFloat(Value=&FloatFromCurve(FloatCurve=DataInterface() {propertyValues {Curve {keys [{arriveTangent=-6.725, interpMode=RCIM_Cubic, leaveTangent=-6.725, tangentMode=RCTM_User, time=0.65, value=1}, <at 1 0 RCIM_Cubic>]}}})))
   &ScaleColor(
    Scale RGB=&VectorFromCurve(VectorCurve=DataInterface() {propertyValues {
       XCurve {keys [<at 0.5 12 RCIM_Cubic>,<at 1 1 RCIM_Cubic>]}
       YCurve {keys [<at 0.5 6 RCIM_Cubic>,<at 1 1 RCIM_Cubic>]}
       ZCurve {keys [<at 0.5 2 RCIM_Cubic>,<at 1 1 RCIM_Cubic>]}
      }})
    ScaleA=true
    ScaleRGB=true
   )
  ]
  rendererClasses [&NiagaraSpriteRendererProperties]
  renderers [{rendererClass=&NiagaraSpriteRendererProperties}]
 }
 Metal_Sparks_Mobile {
  emitterUpdateScript [
   &EmitterState(Life Cycle Mode=::NewEnumerator1, Loop Behavior=::NewEnumerator1, Loop Duration=0.25, UseLoopDelay=false)
   &SpawnBurst_Instantaneous(
    Spawn Count=&MakeIntFromFloat(Float=&Multiply_Float(
      A=&Multiply_Float(A=200, B=$System.ObjectRadiusScaleFactor)
      B=&MakeCustomFloatFromBool(
       Bool=&SetBoolByIntComparison(A=$User.Resource, B=3, Comparison Type=::NewEnumerator4)
       True Float=1
       False Float=0
      )
     ))
    Use Spawn Probability=false
    UseLoopCountLimit=false
   )
  ]
  particleSpawnScript [
   &InitializeParticle(
    Color=<rgba 1 0.3922 0 1>
    Lifetime=&UniformRangedFloat(Minimum=0.1, Maximum=0.45)
    Mass Mode=::NewEnumerator1
    Sprite Size=&UniformRangedVector2D(Minimum=<1 1>, Maximum=<2 2>)
    Sprite Size Mode=::NewEnumerator4
    UsePositionOffset=false
    Write Material Random=true
   ) {module=InitializeParticle001}
   &ShapeLocation(
    Box Midpoint=<0.5 0.5 0>
    Box Size=$User.Bounds
    MaskSpawnGroup=false
    Override Seed=false
    Shape Primitive=::NewEnumerator2
    Surface Only=false
    UseOwnerScale=true
   )
   /Niagara/Modules/Spawn/Location/SphereLocation(MaskSpawnGroup=false, Module.Override Randomness=false, Sphere Radius=5, Surface Only=false)
   &PointForce(
    Force Falloff Distance:NiagaraFloat=$User.Radius
    Force Origin:NiagaraPosition=$User.HitLocation
    Force Strength=&RandomRangeFloat(Minimum=1500, Maximum=2000)
    Module.Use Falloff=true
    Module.UseLinearForce=false
   )
   &LinearForce(
    Coordinate Space:ENiagaraCoordinateSpace=::World
    Force=&Multiply_VectorByFloat(
     Vector=$User.HitVector
     Float=&MakeCustomFloatFromBool(Bool=$User.Detached, True Float=0, False Float=&RandomRangeFloat(Minimum=950, Maximum=1550))
    )
   )
   &AddVelocity(Velocity=&Multiply_Vector(
     A=$Output.SphereLocation.SphereNormal
     B=&VectorFromFloat(Value=&UniformRangedFloat(Minimum=350, Maximum=750))
    )) {module=AddVelocity001}
   &CameraOffset(Camera Offset Amount=25)
   &ApplyInitialForces()
  ]
  particleUpdateScript [
   &ParticleState()
   &AccelerationForce(Acceleration=<0 0 -980>)
   &Drag(Drag=&RandomRangeFloat(Minimum=3, Maximum=10), Use Linear Drag=true, Use Rotational Drag=false)
   &SolveForcesAndVelocity(Clamp Velocity=false, Limit Acceleration=false)
   &ScaleSpriteSizeBySpeed(Max Scale Factor=<1 10>)
   &ScaleSpriteSize(Scale Factor=&Vector2DFromFloat(Value=&FloatFromCurve(FloatCurve=DataInterface() {propertyValues {Curve {keys [{arriveTangent=-6.725, interpMode=RCIM_Cubic, leaveTangent=-6.725, tangentMode=RCTM_User, time=0.65, value=1}, <at 1 0 RCIM_Cubic>]}}})))
   &ScaleColor(
    Scale RGB=&VectorFromCurve(VectorCurve=DataInterface() {propertyValues {
       XCurve {keys [<at 0.5 12 RCIM_Cubic>,<at 1 1 RCIM_Cubic>]}
       YCurve {keys [<at 0.5 6 RCIM_Cubic>,<at 1 1 RCIM_Cubic>]}
       ZCurve {keys [<at 0.5 2 RCIM_Cubic>,<at 1 1 RCIM_Cubic>]}
      }})
    ScaleA=true
    ScaleRGB=true
   )
  ]
  rendererClasses [&NiagaraSpriteRendererProperties]
  renderers [{rendererClass=&NiagaraSpriteRendererProperties}]
 }
 PlayAudioOnSpawnBurst {
  emitterSpawnScript [/SoundLibrary/NiagaraModules/InitializeSoundLibraryPlayer(Sound Library Player=DataInterface() {propertyValues {
      AttachComponentTag=None
      AttachName=None
      bAttachedSound=false
      bAutoDestroy=true
      bForwardToActorOwner=true
      bForwardToAttachedActors=true
      bStopWhenAttachedDestroyed=false
      EventName {tagName=SoundLibrary.Resource.Destruction.Glass}
     }})]
  emitterUpdateScript [
   &EmitterState(Life Cycle Mode=::NewEnumerator1, Loop Behavior=::NewEnumerator1, Loop Duration=0.25, UseLoopDelay=false)
   &SpawnBurst_Instantaneous(Use Spawn Probability=false, UseLoopCountLimit=true)
  ]
  particleSpawnScript [
   &InitializeParticle(UsePositionOffset=false, Write Material Random=true)
   /SoundLibrary/NiagaraModules/PlaySoundLibraryEvent(
    Audio Playback Enabled:NiagaraBool=&SetBoolByIntComparison(A=$User.Resource, B=5, Comparison Type=::NewEnumerator4)
    Audio Playback Position:NiagaraPosition=$Particles.Position
    Destructed Actor Bounds:Vector3f=$User.Scale
   )
  ]
  particleUpdateScript [&ParticleState()]
  rendererClasses []
  renderers []
 }
 Resource_Mobile {
  emitterUpdateScript [
   &EmitterState(Life Cycle Mode=::NewEnumerator1, Loop Behavior=::NewEnumerator1, Loop Duration=0.1, UseLoopDelay=false)
   &SpawnBurst_Instantaneous(
    Spawn Count=&MakeIntFromFloat(Float=&Multiply_Float(A=40, B=$System.ObjectRadiusScaleFactor))
    Use Spawn Probability=false
    UseLoopCountLimit=false
   )
  ]
  particleSpawnScript [
   &InitializeParticle(
    Lifetime Max=&MakeCustomFloatFromBool(
     Bool=&SetBoolByIntComparison(A=$User.Resource, B=4, Comparison Type=::NewEnumerator4)
     True Float=4
     False Float=2
    )
    Lifetime Min:NiagaraFloat=&MakeCustomFloatFromBool(
     Bool=&SetBoolByIntComparison(A=$User.Resource, B=4, Comparison Type=::NewEnumerator4)
     True Float=2
     False Float=1
    )
    Lifetime Mode=::NewEnumerator1
    Mass Max=1
    Mass Min=0.5
    Mass Mode=::NewEnumerator3
    Mesh Scale Mode=::NewEnumerator3
    Mesh Uniform Scale=&self:MeshDynamicScalar(Mesh Scale:NiagaraFloat=1, Object Sphere Radius:NiagaraFloat=$User.Radius, Comparison Value:NiagaraFloat=384)
    UsePositionOffset=false
    Write Material Random=true
   )
   &self:ResourceSetup(
    Default Mass 0:NiagaraFloat=&RandomRangeFloat(Minimum=2, Maximum=7)
    Default Mass 1:NiagaraFloat=&RandomRangeFloat(Minimum=3, Maximum=9)
    Default Mass 2:NiagaraFloat=&RandomRangeFloat(Minimum=5, Maximum=17)
    Default Scale 0:Vector3f=&VectorFromFloat(Value=&RandomRangeFloat(Minimum=0.5, Maximum=1.5))
    Default Scale 1:Vector3f=&VectorFromFloat(Value=&RandomRangeFloat(Minimum=0.75, Maximum=1.25))
    Default Scale 2:Vector3f=&VectorFromFloat(Value=&RandomRangeFloat(Minimum=1.25, Maximum=1.6))
    "Foliage Color (Fallback)":LinearColor=<rgba 0 0.5 0 1>
    Foliage Mass:NiagaraFloat=&RandomRangeFloat(Minimum=1, Maximum=2.5)
    Foliage Scale:Vector3f=&VectorFromFloat(Value=&RandomRangeFloat(Minimum=3, Maximum=8))
    Medium Mesh=0.96
    "Metal Color (Fallback)":LinearColor=<rgba 0.1 0.1 0.1 1>
    Metal Mass 0:NiagaraFloat=&RandomRangeFloat(Minimum=2.5, Maximum=7)
    Metal Mass 1:NiagaraFloat=&RandomRangeFloat(Minimum=3.5, Maximum=9)
    Metal Mass 2:NiagaraFloat=&RandomRangeFloat(Minimum=5.5, Maximum=17)
    Metal Mass 3:NiagaraFloat=&RandomRangeFloat(Minimum=3.5, Maximum=9)
    Metal Mass 4:NiagaraFloat=&RandomRangeFloat(Minimum=5.5, Maximum=17)
    Metal Scale 0:Vector3f=&VectorFromFloat(Value=&RandomRangeFloat(Minimum=0.5, Maximum=1.6))
    Metal Scale 1:Vector3f=&VectorFromFloat(Value=&RandomRangeFloat(Minimum=0.5, Maximum=1.2))
    Metal Scale 2:Vector3f=&VectorFromFloat(Value=&RandomRangeFloat(Minimum=0.8, Maximum=1.2))
    Metal Scale 3:Vector3f=&VectorFromFloat(Value=&RandomRangeFloat(Minimum=0.8, Maximum=1.2))
    Metal Scale 4:Vector3f=&VectorFromFloat(Value=&RandomRangeFloat(Minimum=0.5, Maximum=1.6))
    Module.Override Randomness=false
    PhysicalSurface:NiagaraInt32=$User.Resource
    Small Mesh=0.5
    Spawn Location:NiagaraPosition=$Particles.Position
    "Stone Color (Fallback)":LinearColor=<rgba 0.25 0.25 0.25 1>
    Stone Mass 0:NiagaraFloat=&RandomRangeFloat(Minimum=2, Maximum=7)
    Stone Mass 1:NiagaraFloat=&RandomRangeFloat(Minimum=3, Maximum=9)
    Stone Mass 2:NiagaraFloat=&RandomRangeFloat(Minimum=5, Maximum=17)
    Stone Scale 0:Vector3f=&VectorFromFloat(Value=&RandomRangeFloat(Minimum=0.5, Maximum=1.5))
    Stone Scale 1:Vector3f=&VectorFromFloat(Value=&RandomRangeFloat(Minimum=0.75, Maximum=1.25))
    Stone Scale 2:Vector3f=&VectorFromFloat(Value=&RandomRangeFloat(Minimum=1.25, Maximum=1.6))
    "Wood Color (Fallback)":LinearColor=<rgba 0.5 0.3001 0.1783 1>
    Wood Mass 0:NiagaraFloat=&RandomRangeFloat(Minimum=1.5, Maximum=8)
    Wood Mass 1:NiagaraFloat=&RandomRangeFloat(Minimum=2.5, Maximum=9)
    Wood Mass 2:NiagaraFloat=&RandomRangeFloat(Minimum=4.5, Maximum=17)
    Wood Scale 0:Vector3f=&VectorFromFloat(Value=&RandomRangeFloat(Minimum=0.5, Maximum=1.3))
    Wood Scale 1:Vector3f=&VectorFromFloat(Value=&RandomRangeFloat(Minimum=0.5, Maximum=1.6))
    Wood Scale 2:Vector3f=&VectorFromFloat(Value=&RandomRangeFloat(Minimum=1, Maximum=1.6))
    bisMobileRender=true
    foliageVisibilityTag=4
    metalVisibilityTag=3
    stoneVisibilityTag=2
    woodVisibilityTag=1
   ) {module=ResourceSetup}
   &ShapeLocation(Box Size=$User.Bounds, MaskSpawnGroup=false, Override Seed=false, Shape Primitive=::NewEnumerator2, Surface Only=false, UseOwnerScale=true)
   &InitialMeshOrientation(Use Rotation Vector=false)
   &PointForce(
    Force Falloff Distance:NiagaraFloat=$User.Radius
    Force Origin:NiagaraPosition=$User.HitLocation
    Force Strength=&MakeCustomFloatFromBool(
     Bool=$User.bIsInstantDeath
     True Float=&Multiply_Float(A=&RandomRangeFloat(Minimum=1000, Maximum=1500), B=$System.ImpulseScale)
     False Float=&Multiply_Float(A=&RandomRangeFloat(Minimum=1750, Maximum=2250), B=$System.ImpulseScale)
    )
    Module.Use Falloff=true
    Module.UseLinearForce=false
   )
   &LinearForce(
    Coordinate Space:ENiagaraCoordinateSpace=::World
    Force=&Multiply_VectorByFloat(
     Vector=&Multiply_Vector(A=$User.HitVector, B=<1 1 0>)
     Float=&MakeCustomFloatFromBool(
      Bool=$User.bIsInstantDeath
      True Float=&Multiply_Float(A=&RandomRangeFloat(Minimum=1250, Maximum=1750), B=$System.ImpulseScale)
      False Float=&Multiply_Float(A=&RandomRangeFloat(Minimum=1500, Maximum=2000), B=$System.ImpulseScale)
     )
    )
   )
   &ApplyInitialForces()
  ]
  particleUpdateScript [
   &ParticleState()
   &GravityForce(Gravity=&Multiply_VectorByFloat(
     Vector=<0 0 -980>
     Float=&FloatFromCurve(FloatCurve=DataInterface() {propertyValues {
        bOptimizeLUT=false
        Curve {keys [<at 0.05 0 RCIM_Cubic>, {arriveTangent=45.9003, interpMode=RCIM_Cubic, leaveTangent=45.9003, tangentMode=RCTM_User, time=0.1, value=1}]}
       }})
    ))
   &Drag(
    Drag=&self:Gravity_Drag_Compensation(Drag Falloff:NiagaraFloat=2, Gravity Direction:Vector3f=<0 0 -1>)
    Use Linear Drag=true
    Use Rotational Drag=true
   )
   &UpdateMeshOrientation(Rotation Vector=&RandomRangeVector(Minimum=<-1 -1 -1>, Maximum=<1 1 1>))
   &SolveForcesAndVelocity(Clamp Velocity=true, Limit Acceleration=true, Speed Limit=2500)
   &ScaleMeshSize(Scale Factor=&VectorFromFloat(Value=&FloatFromCurve(FloatCurve=DataInterface() {propertyValues {Curve {keys [
          <at 0 0 RCIM_Constant>
          <at 0.05 1 RCIM_Constant>
          {interpMode=RCIM_Cubic, leaveTangentWeight=0.3333, tangentMode=RCTM_User, time=0.75, value=1}
          {arriveTangentWeight=0.3333, interpMode=RCIM_Cubic, tangentMode=RCTM_User, time=1, value=0}
         ]}}})))
   &DynamicMaterialParameters(
    Index 0 Param 1=&MakeCustomFloatFromBool(
     Bool=$User.bIsCurie
     True Float=&FloatFromCurve(FloatCurve=DataInterface() {propertyValues {Curve {keys [
          {arriveTangent=-9.3075, interpMode=RCIM_Cubic, leaveTangent=-9.3075, tangentMode=RCTM_User, time=0, value=2}
          {arriveTangent=-2.0574, interpMode=RCIM_Cubic, leaveTangent=-2.0574, tangentMode=RCTM_User, time=0.2, value=0.625}
          <at 0.75 0 RCIM_Cubic>
         ]}}})
     False Float=0
    )
    Index 0 Param 2=&MakeCustomFloatFromBool(Bool=$User.bIsCurie, True Float=1, False Float=0)
    Index 0 Param 4=&RandomRangeFloat(Minimum=0, Maximum=1)
    Module.Index0Float0Write=true
    Module.Index0Float1Write=true
    Module.Index0Float2Write=true
    Module.Index0Float3Write=true
   )
  ]
  rendererClasses [&NiagaraMeshRendererProperties]
  renderers [
   {rendererClass=&NiagaraMeshRendererProperties, rendererIndex=0}
   {rendererClass=&NiagaraMeshRendererProperties, rendererIndex=1}
   {rendererClass=&NiagaraMeshRendererProperties, rendererIndex=2}
   {rendererClass=&NiagaraMeshRendererProperties, rendererIndex=3}
   {rendererClass=&NiagaraMeshRendererProperties, rendererIndex=4}
  ]
 }
 Resource_Query {
  simTarget=GPUComputeSim
  emitterSpawnScript [&self:RigidBodyQuerySetup(
    Bounds:Vector3f=&Multiply_VectorByFloat(Vector=$User.Bounds, Float=1.125)
    Rigid Mesh Collision Query=DataInterface() {propertyValues {
      GlobalSearchAllowed=false
      GlobalSearchFallback_Unscripted=false
      GlobalSearchForced=false
      MaxNumPrimitives=100
      OnlyUseMoveable=false
      UseComplexCollisions=false
      bFilterByObjectType=false
      ActorTags []
      ComponentTags []
      SourceActors []
     }}
   ) {module=RigidBodyQuerySetup}]
  emitterUpdateScript [
   &EmitterState(Life Cycle Mode=::NewEnumerator1, Loop Behavior=::NewEnumerator1, Loop Duration=0.1, UseLoopDelay=false)
   &SpawnBurst_Instantaneous(
    Spawn Count=&MakeIntFromFloat(Float=&Multiply_Float(A=100, B=$System.ObjectRadiusScaleFactor))
    Use Spawn Probability=false
    UseLoopCountLimit=false
   )
  ]
  particleSpawnScript [
   &InitializeParticle(
    Color Mode=::NewEnumerator0
    Lifetime Max=&MakeCustomFloatFromBool(
     Bool=&SetBoolByIntComparison(A=$User.Resource, B=4, Comparison Type=::NewEnumerator4)
     True Float=4
     False Float=2
    )
    Lifetime Min:NiagaraFloat=&MakeCustomFloatFromBool(
     Bool=&SetBoolByIntComparison(A=$User.Resource, B=4, Comparison Type=::NewEnumerator4)
     True Float=2
     False Float=1
    )
    Lifetime Mode=::NewEnumerator1
    Mass Max=1
    Mass Min=0.5
    Mass Mode=::NewEnumerator3
    Mesh Scale Mode=::NewEnumerator3
    Mesh Uniform Scale=&self:MeshDynamicScalar(Mesh Scale:NiagaraFloat=1, Object Sphere Radius:NiagaraFloat=$User.Radius, Comparison Value:NiagaraFloat=384)
    UsePositionOffset=false
    Write Material Random=true
   )
   &ShapeLocation(
    Box Size=&Multiply_VectorByFloat(Vector=$User.Bounds, Float=2)
    MaskSpawnGroup=false
    Override Seed=false
    Shape Primitive=::NewEnumerator2
    Surface Only=false
    UseOwnerScale=true
   )
   &self:RigidBodyQuerySampling(MaxDistance=100, Rigid Mesh Collision Query:NiagaraDataInterfaceRigidMeshCollisionQuery=$Emitter.Rigid Mesh Collision Query) {module=RigidBodyQuerySampling}
   /SetVariables(Particles.Position=$Particles.Closest Position) {isSetParameters=true}
   &self:ResourceSetup(
    Default Mass 0:NiagaraFloat=&RandomRangeFloat(Minimum=2, Maximum=7)
    Default Mass 1:NiagaraFloat=&RandomRangeFloat(Minimum=3, Maximum=9)
    Default Mass 2:NiagaraFloat=&RandomRangeFloat(Minimum=5, Maximum=17)
    Default Scale 0:Vector3f=&VectorFromFloat(Value=&RandomRangeFloat(Minimum=0.25, Maximum=1.333))
    Default Scale 1:Vector3f=&VectorFromFloat(Value=&RandomRangeFloat(Minimum=0.25, Maximum=1.333))
    Default Scale 2:Vector3f=&VectorFromFloat(Value=&RandomRangeFloat(Minimum=1, Maximum=1.5))
    "Foliage Color (Fallback)":LinearColor=<rgba 0 0.5 0 1>
    Foliage Mass:NiagaraFloat=&RandomRangeFloat(Minimum=1, Maximum=2.5)
    Foliage Scale:Vector3f=&VectorFromFloat(Value=&RandomRangeFloat(Minimum=3, Maximum=8))
    Medium Mesh=0.96
    "Metal Color (Fallback)":LinearColor=<rgba 0.1 0.1 0.1 1>
    Metal Mass 0:NiagaraFloat=&RandomRangeFloat(Minimum=2.5, Maximum=7)
    Metal Mass 1:NiagaraFloat=&RandomRangeFloat(Minimum=3.5, Maximum=9)
    Metal Mass 2:NiagaraFloat=&RandomRangeFloat(Minimum=5.5, Maximum=17)
    Metal Mass 3:NiagaraFloat=&RandomRangeFloat(Minimum=3.5, Maximum=9)
    Metal Mass 4:NiagaraFloat=&RandomRangeFloat(Minimum=5.5, Maximum=17)
    Metal Scale 0:Vector3f=&VectorFromFloat(Value=&RandomRangeFloat(Minimum=0.25, Maximum=1.333))
    Metal Scale 1:Vector3f=&VectorFromFloat(Value=&RandomRangeFloat(Minimum=0.25, Maximum=1))
    Metal Scale 2:Vector3f=&VectorFromFloat(Value=&RandomRangeFloat(Minimum=0.8, Maximum=1))
    Metal Scale 3:Vector3f=&VectorFromFloat(Value=&RandomRangeFloat(Minimum=0.8, Maximum=1.2))
    Metal Scale 4:Vector3f=&VectorFromFloat(Value=&RandomRangeFloat(Minimum=0.333, Maximum=0.5))
    Module.Override Randomness=false
    PhysicalSurface:NiagaraInt32=$User.Resource
    Small Mesh=0.5
    Spawn Location:NiagaraPosition=$Particles.Position
    "Stone Color (Fallback)":LinearColor=<rgba 0.25 0.25 0.25 1>
    Stone Mass 0:NiagaraFloat=&RandomRangeFloat(Minimum=2, Maximum=7)
    Stone Mass 1:NiagaraFloat=&RandomRangeFloat(Minimum=3, Maximum=9)
    Stone Mass 2:NiagaraFloat=&RandomRangeFloat(Minimum=5, Maximum=17)
    Stone Scale 0:Vector3f=&VectorFromFloat(Value=&RandomRangeFloat(Minimum=0.25, Maximum=1.333))
    Stone Scale 1:Vector3f=&VectorFromFloat(Value=&RandomRangeFloat(Minimum=0.25, Maximum=1.333))
    Stone Scale 2:Vector3f=&VectorFromFloat(Value=&RandomRangeFloat(Minimum=1, Maximum=1.5))
    "Wood Color (Fallback)":LinearColor=<rgba 0.5 0.3001 0.1783 1>
    Wood Mass 0:NiagaraFloat=&RandomRangeFloat(Minimum=1.5, Maximum=8)
    Wood Mass 1:NiagaraFloat=&RandomRangeFloat(Minimum=2.5, Maximum=9)
    Wood Mass 2:NiagaraFloat=&RandomRangeFloat(Minimum=4.5, Maximum=17)
    Wood Scale 0:Vector3f=&VectorFromFloat(Value=&RandomRangeFloat(Minimum=0.25, Maximum=1))
    Wood Scale 1:Vector3f=&VectorFromFloat(Value=&RandomRangeFloat(Minimum=0.25, Maximum=1.333))
    Wood Scale 2:Vector3f=&VectorFromFloat(Value=&RandomRangeFloat(Minimum=0.9, Maximum=1.5))
    foliageVisibilityTag=4
    metalVisibilityTag=3
    stoneVisibilityTag=2
    woodVisibilityTag=1
   ) {module=ResourceSetup}
   &PointForce(
    Force Falloff Distance:NiagaraFloat=$User.Radius
    Force Origin:NiagaraPosition=$User.HitLocation
    Force Strength=&MakeCustomFloatFromBool(
     Bool=$User.bIsInstantDeath
     True Float=&Multiply_Float(A=&RandomRangeFloat(Minimum=1000, Maximum=1500), B=$System.ImpulseScale)
     False Float=&Multiply_Float(A=&RandomRangeFloat(Minimum=1750, Maximum=2250), B=$System.ImpulseScale)
    )
    Module.Use Falloff=true
    Module.UseLinearForce=false
   )
   &LinearForce(
    Coordinate Space:ENiagaraCoordinateSpace=::World
    Force=&Multiply_VectorByFloat(
     Vector=&Multiply_Vector(A=$User.HitVector, B=<1 1 0>)
     Float=&MakeCustomFloatFromBool(
      Bool=$User.bIsInstantDeath
      True Float=&Multiply_Float(A=&RandomRangeFloat(Minimum=1250, Maximum=1750), B=$System.ImpulseScale)
      False Float=&Multiply_Float(A=&RandomRangeFloat(Minimum=1500, Maximum=2000), B=$System.ImpulseScale)
     )
    )
   )
   &ApplyInitialForces()
   &InitialMeshOrientation(
    Mesh Axis To Orient=<0 0 1>
    Mesh Orientation Mode=::NewEnumerator2
    Orientation Vector=$Particles.Closest Normal
    Rotation=&RandomRangeVector(Minimum=<0 0 0>, Maximum=<0 0 1>)
    Use Rotation Vector=true
   )
  ]
  particleUpdateScript [
   &ParticleState()
   &KillParticles(Kill Particles=&self:Remove_Black_Particles())
   &GravityForce(Gravity=&Multiply_VectorByFloat(
     Vector=<0 0 -980>
     Float=&FloatFromCurve(FloatCurve=DataInterface() {propertyValues {
        bOptimizeLUT=false
        Curve {keys [<at 0.05 0 RCIM_Cubic>, {arriveTangent=45.9003, interpMode=RCIM_Cubic, leaveTangent=45.9003, tangentMode=RCTM_User, time=0.1, value=1}]}
       }})
    ))
   &Drag(
    Drag=&self:Gravity_Drag_Compensation(Drag Falloff:NiagaraFloat=2, Gravity Direction:Vector3f=<0 0 -1>)
    Use Linear Drag=true
    Use Rotational Drag=true
   )
   &Collision(
    Age Colliding Particles=false
    Collision Enabled=&RandomBool()
    Friction During a Bounce=0.2
    GPU Collision Type=::NewEnumerator2
    Mesh Dimensions:Vector3f=&self:Spawned_Mesh_Dimensions(
     Physical Surface:NiagaraInt32=$User.Resource
     Wood Mesh 0:Vector3f=<65 29 34>
     Wood Mesh 1:Vector3f=<96 21 3>
     Wood Mesh 2:Vector3f=<69 17 18>
     Stone Mesh 0:Vector3f=<18 16 21>
     Stone Mesh 1:Vector3f=<43 39 42>
     Stone Mesh 2:Vector3f=<47 60 46>
     Metal Mesh 0:Vector3f=<32 32 6>
     Metal Mesh 1:Vector3f=<52 52 13>
     Metal Mesh 2:Vector3f=<16 7 9>
     Metal Mesh 3:Vector3f=<34 14 17>
     Metal Mesh 4:Vector3f=<46 46 0>
     Default Mesh 0:Vector3f=<18 16 21>
     Default Mesh 1:Vector3f=<43 39 42>
     Default Mesh 2:Vector3f=<47 60 46>
     Foliage Mesh:Vector3f=<4 2 0>
    )
    Module.Control Rest State Via Surface Normal=false
    Module.EnableMaxCollisionCount=false
    Module.Randomize Collision Normal Vector=true
    Particle Radius Scale=0.1
    Radius Calculation Type=::NewEnumerator1
    Restitution=0.15
   )
   &SolveForcesAndVelocity(Clamp Velocity=true, Limit Acceleration=true, Speed Limit=2500)
   &UpdateMeshOrientation(Rotation Vector=&MakeCustomVectorFromBool(
     Bool=$Particles.HasCollided
     True Vector=<0 0 0>
     False Vector=&RandomRangeVector(Minimum=<-1 -1 -1>, Maximum=<1 1 1>)
    ))
   &ScaleMeshSize(Scale Factor=&VectorFromFloat(Value=&FloatFromCurve(FloatCurve=DataInterface() {propertyValues {Curve {keys [
          <at 0 0 RCIM_Constant>
          <at 0.05 1 RCIM_Constant>
          {interpMode=RCIM_Cubic, leaveTangentWeight=0.3333, tangentMode=RCTM_User, time=0.75, value=1}
          {arriveTangentWeight=0.3333, interpMode=RCIM_Cubic, tangentMode=RCTM_User, time=1, value=0}
         ]}}})))
   &DynamicMaterialParameters(
    Index 0 Param 1=&MakeCustomFloatFromBool(
     Bool=$User.bIsCurie
     True Float=&FloatFromCurve(FloatCurve=DataInterface() {propertyValues {Curve {keys [
          {arriveTangent=-9.3075, interpMode=RCIM_Cubic, leaveTangent=-9.3075, tangentMode=RCTM_User, time=0, value=2}
          {arriveTangent=-2.0574, interpMode=RCIM_Cubic, leaveTangent=-2.0574, tangentMode=RCTM_User, time=0.2, value=0.625}
          <at 0.75 0 RCIM_Cubic>
         ]}}})
     False Float=0
    )
    Index 0 Param 2=&MakeCustomFloatFromBool(Bool=$User.bIsCurie, True Float=1, False Float=0)
    Index 0 Param 4=&RandomRangeFloat(Minimum=0, Maximum=1)
    Module.Index0Float0Write=true
    Module.Index0Float1Write=true
    Module.Index0Float2Write=true
    Module.Index0Float3Write=true
   )
  ]
  rendererClasses [&NiagaraMeshRendererProperties]
  renderers [
   {rendererClass=&NiagaraMeshRendererProperties, rendererIndex=0}
   {rendererClass=&NiagaraMeshRendererProperties, rendererIndex=1}
   {rendererClass=&NiagaraMeshRendererProperties, rendererIndex=2}
   {rendererClass=&NiagaraMeshRendererProperties, rendererIndex=3}
   {rendererClass=&NiagaraMeshRendererProperties, rendererIndex=4}
  ]
 }
 Smoke {
  emitterUpdateScript [
   &EmitterState(Life Cycle Mode=::NewEnumerator1, Loop Behavior=::NewEnumerator1, Loop Duration=0.25, UseLoopDelay=false)
   &SpawnBurst_Instantaneous(
    Spawn Count=&MakeCustomIntFromBool(
     Bool=&Boolean_OrOperation(
      Bool1=$User.bIsInstantDeath
      Bool2=&SetBoolByIntComparison(A=$User.Resource, B=5, Comparison Type=::NewEnumerator4)
     )
     True INT=0
     False INT=&RandomRangeInt(Minimum=1, Maximum=3)
    )
    Use Spawn Probability=false
    UseLoopCountLimit=false
   )
  ]
  particleSpawnScript [
   &InitializeParticle(
    Color=<rgba 1 1 1 0.15>
    Lifetime Max=0.35
    Lifetime Min=0.3
    Lifetime Mode=::NewEnumerator1
    Position=$User.HitLocation
    Position Mode=::NewEnumerator1
    Sprite Rotation Angle=&RandomRangeFloat(Minimum=-45, Maximum=45)
    Sprite Rotation Mode=::NewEnumerator0
    Sprite Size Mode=::NewEnumerator3
    Uniform Sprite Size=&RandomRangeFloat(Minimum=375, Maximum=500)
    UsePositionOffset=false
    Write Material Random=true
   )
   &CameraOffset(Camera Offset Amount=30)
  ]
  particleUpdateScript [
   &ParticleState()
   &ScaleColor(
    Scale Alpha=&FloatFromCurve(FloatCurve=DataInterface() {propertyValues {Curve {keys [<at 0.5 1 RCIM_Cubic>,<at 1 0 RCIM_Cubic>]}}})
    ScaleA=true
    ScaleRGB=true
   )
   &ScaleSpriteSize(Scale Factor=&Vector2DFromFloat(Value=&FloatFromCurve(FloatCurve=DataInterface() {propertyValues {Curve {keys [
          {arriveTangent=8.1876, interpMode=RCIM_Cubic, leaveTangent=5.3162, tangentMode=RCTM_User, time=0, value=0.15}
          {arriveTangent=1.0025, interpMode=RCIM_Cubic, leaveTangent=1.0025, tangentMode=RCTM_User, time=0.161, value=0.8}
          {arriveTangent=0.1638, interpMode=RCIM_Cubic, leaveTangent=0.1638, tangentMode=RCTM_User, time=1.0463, value=1.25}
         ]}}})))
   &DynamicMaterialParameters(
    Index 0 Param 1=&FloatFromCurve(FloatCurve=DataInterface() {propertyValues {Curve {keys [<at 0 0>,<at 1 1>]}}})
    Index 0 Param 2=1
    Module.Index0Float0Write=true
    Module.Index0Float1Write=true
    Module.Index0Float2Write=false
    Module.Index0Float3Write=false
   )
  ]
  rendererClasses [&NiagaraSpriteRendererProperties]
  renderers [{rendererClass=&NiagaraSpriteRendererProperties}]
 }
 Toon_Lines {
  emitterUpdateScript [
   &EmitterState(Life Cycle Mode=::NewEnumerator1, Loop Behavior=::NewEnumerator1, Loop Duration=1, UseLoopDelay=false)
   &SpawnBurst_Instantaneous(
    Spawn Count=&MakeCustomIntFromBool(
     Bool=&Boolean_OrOperation(
      Bool1=$User.bIsInstantDeath
      Bool2=&SetBoolByIntComparison(A=$User.Resource, B=5, Comparison Type=::NewEnumerator4)
     )
     True INT=0
     False INT=1
    )
    Use Spawn Probability=false
    UseLoopCountLimit=false
   )
  ]
  particleSpawnScript [
   &InitializeParticle(
    Color=<rgba 0 0 0 10>
    Lifetime=0.3
    Position=$User.HitLocation
    Position Mode=::NewEnumerator1
    Sprite Rotation Mode=::NewEnumerator2
    Sprite Size Mode=::NewEnumerator3
    Uniform Sprite Size=&RandomRangeFloat(Minimum=375, Maximum=500)
    UsePositionOffset=false
    Write Material Random=true
   )
   &CameraOffset(Camera Offset Amount=30)
  ]
  particleUpdateScript [
   &ParticleState()
   &ScaleColor(ScaleA=true, ScaleRGB=true)
   &ScaleSpriteSize(Scale Factor=&Vector2DFromFloat(Value=&FloatFromCurve(FloatCurve=DataInterface() {propertyValues {Curve {keys [{arriveTangent=3.9086, interpMode=RCIM_Cubic, leaveTangent=3.9086, tangentMode=RCTM_User, time=0, value=0}, <at 1 1.4 RCIM_Cubic>]}}})))
   &DynamicMaterialParameters(
    Index 0 Param 1=&FloatFromCurve(FloatCurve=DataInterface() {propertyValues {Curve {keys [<at 0 1 RCIM_Cubic>, {arriveTangent=-2.742, interpMode=RCIM_Cubic, leaveTangent=-2.742, tangentMode=RCTM_User, time=1, value=0}]}}})
    Index 0 Param 2=&FloatFromCurve(FloatCurve=DataInterface() {propertyValues {Curve {keys [<at 0 0 RCIM_Cubic>, {arriveTangent=2.4833, interpMode=RCIM_Cubic, leaveTangent=2.4833, tangentMode=RCTM_User, time=1, value=1}]}}})
    Module.Index0Float0Write=true
    Module.Index0Float1Write=true
    Module.Index0Float2Write=false
    Module.Index0Float3Write=false
   )
  ]
  rendererClasses [&NiagaraSpriteRendererProperties]
  renderers [{rendererClass=&NiagaraSpriteRendererProperties}]
 }
}
summary {
 systemName=NS_Enviromental_Destruction_Master_Toon
 emitters [
  {emitterName=Metal_Sparks, simTarget=GPUComputeSim, rendererClasses [&NiagaraSpriteRendererProperties]}
  {emitterName=Smoke, rendererClasses [&NiagaraSpriteRendererProperties]}
  {emitterName=PlayAudioOnSpawnBurst, rendererClasses []}
  {emitterName=Metal_Sparks_Mobile, rendererClasses [&NiagaraSpriteRendererProperties]}
  {bEnabled=false, emitterName=Audio_Collision, rendererClasses []}
  {emitterName=Glass, simTarget=GPUComputeSim, rendererClasses [&NiagaraMeshRendererProperties]}
  {emitterName=Toon_Lines, rendererClasses [&NiagaraSpriteRendererProperties]}
  {emitterName=Resource_Mobile, rendererClasses [&NiagaraMeshRendererProperties]}
  {emitterName=Resource_Query, simTarget=GPUComputeSim, rendererClasses [&NiagaraMeshRendererProperties]}
 ]
 userVariables [
  User.bIsCurie {type=NiagaraBool}
  User.bIsInstantDeath {type=NiagaraBool}
  User.Bounds {type=Vector3f, description="", defaultValue {struct=Vector3f, value=<1 500 500>}}
  User.DamageDealt {type=NiagaraFloat, defaultValue=50, description=""}
  User.Detached {type=NiagaraBool}
  User.HitLocation {type=NiagaraPosition, description="", defaultValue {}}
  User.HitVector {type=Vector3f, description="", defaultValue {struct=Vector3f, value=<1 0 0>}}
  User.Radius {type=NiagaraFloat, defaultValue=100, description=""}
  User.Resource {type=NiagaraInt32, defaultValue=4}
  User.Scale {type=Vector3f, description="", defaultValue {struct=Vector3f, value=<1 1 1>}}
 ]
}
