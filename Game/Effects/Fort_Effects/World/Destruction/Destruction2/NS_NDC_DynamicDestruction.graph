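# Niagara system for dynamic destruction debris and effects, spawning from hit events read through a Niagara Data Channel.
# Visible emitters add mesh chunks, dust/smoke, metal sparks and Curie sparks, switched or filtered by resource type and Curie flag.
# GPU chunks and sparks snap to static-mesh distance-field surfaces; chunks sample GBuffer color and take hit-driven impulses.
# Mobile variants cut spawn counts and skip surface snapping and collision.
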NiagaraSystem /Game/Effects/Fort_Effects/World/Destruction/Destruction2/NS_NDC_DynamicDestruction
dependencies {
 usedDataInterfaces [
  &NiagaraDataInterfaceCurve
  /Script/Niagara.NiagaraDataInterfaceStaticMeshIndirect
  &NiagaraDataInterfaceGBuffer
  &NiagaraDataInterfaceArrayInt32
  &NiagaraDataInterfaceSpriteRendererInfo
  &NiagaraDataInterfaceVectorCurve
  /Script/Niagara.NiagaraDataInterfaceDataChannelRead
 ]
 usedDynamicInputs []
 usedModules [
  &SystemState
  &self:SystemSpawnScript.SystemScriptSource.SystemScriptGraph.NiagaraNodeAssignment_0.SetVariables_#1
  /Niagara/Modules/System/CompleteIfUnused
 ]
}
emitters {
 Curie_Sparks {
  simTarget=GPUComputeSim
  emitterUpdateScript [
   &EmitterState()
   &self:Spawn_NDCDynamicDestruction(Data Channel:NiagaraDataInterfaceDataChannelRead=$System.NDCReader, Max Count=600, Min Count=300, Scale Spawn Count=false) {module=Spawn_NDCDynamicDestruction}
  ]
  particleSpawnScript [
   &self:Initialize_NDCDynamicDestruction_Simple() {module=Initialize_NDCDynamicDestruction_Simple}
   &InitializeParticle(
    Color=<rgba 30 13.125 3 1>
    Lifetime Max=1
    Lifetime Min=0.15
    Lifetime Mode=::NewEnumerator1
    Position Mode=::NewEnumerator0
    Sprite Size Max=<3.5 2>
    Sprite Size Min=<2.5 1>
    Sprite Size Mode=::NewEnumerator2
    Write Material Random=true
   )
   &ShapeLocation(
    Additional Quaternion Rotation=false
    Box Size=&Multiply_VectorByFloat(Vector=$Particles.BoundsExtents, Float=2)
    MaskSpawnGroup=false
    Override Seed=false
    Rotation Mode=::NewEnumerator3
    Rotation Quaternion=$Particles.Orientation
    Shape Primitive=::NewEnumerator2
    Surface Only=false
    UseOwnerScale=true
   ) {module=ShapeLocation001}
   &self:Place_On_DF_Surface(KillOnFailure=true, Max Distance=2000, StaticMeshID:NiagaraSharedDataID_StaticMesh=$Particles.StaticMeshID, Use Max Distance=true) {module=Place_On_DF_Surface}
  ]
  particleUpdateScript [
   &ParticleState()
   &KillParticles(Kill Particles=&Boolean_NotOperation(Bool=$Particles.bisCurie))
   &GravityForce(Gravity=<0 0 -600>)
   /Niagara/Modules/Update/Forces/CurlNoiseForce(
    Noise Frequency=300
    Noise Strength=&Multiply_Float(
     A=$Particles.MaterialRandom
     B=&FloatFromCurve(
      FloatCurve=DataInterface() {propertyValues {Curve {keys [
          <at 0 4147.9209 RCIM_Cubic>
          {arriveTangent=998.7806, interpMode=RCIM_Cubic, leaveTangent=998.7806, time=0.897, value=2308.9619}
          {arriveTangent=839.8494, interpMode=RCIM_Cubic, leaveTangent=839.8494, time=2.2666, value=6411.7881}
          {arriveTangent=818.7506, interpMode=RCIM_Cubic, leaveTangent=818.7506, time=3.6727, value=4640.1284}
          {arriveTangent=-197.3881, interpMode=RCIM_Cubic, leaveTangent=-197.3881, time=4.7165, value=8417.627}
          {arriveTangent=-812.5138, interpMode=RCIM_Cubic, leaveTangent=-812.5138, time=6.6551, value=4051.4272}
          {arriveTangent=379.0706, interpMode=RCIM_Cubic, leaveTangent=379.0706, time=9.3392, value=4661.6182}
          {arriveTangent=-88.1153, interpMode=RCIM_Cubic, leaveTangent=-88.1153, time=11.8318, value=6013.7651}
          <at 15.169 4147.9209 RCIM_Cubic>
         ]}}}
      CurveIndex=&Add_Float(A=$Engine.System.Age, B=$Particles.Age)
      Scale Curve=0.7
     )
    )
    Pan Noise=true
    Pan Noise Field=<0 0 1000>
   )
   &Drag(Drag=0.8, Use Linear Drag=true, Use Rotational Drag=false)
   &Collision(
    Age Colliding Particles=false
    Friction=0.35
    GPU Collision Type=::NewEnumerator2
    Module.Control Rest State Via Surface Normal=false
    Module.EnableMaxCollisionCount=false
    Module.Randomize Collision Normal Vector=false
    Restitution=0.4
   )
   &SolveForcesAndVelocity(Clamp Velocity=false, Limit Acceleration=false)
   &ScaleSpriteSizeBySpeed(Max Scale Factor=<1 10>, Velocity Threshold=1300)
   &ScaleColor(
    Scale Alpha=&FloatFromCurve(FloatCurve=DataInterface() {propertyValues {Curve {keys [<at 0 0>,<at 0.1997 0.9273>,<at 1.0011 0.002>]}}})
    ScaleA=true
    ScaleRGB=true
   )
  ]
  rendererClasses [&NiagaraSpriteRendererProperties]
  renderers [{rendererClass=&NiagaraSpriteRendererProperties}]
 }
 Curie_Sparks_Mobile {
  emitterUpdateScript [
   &EmitterState()
   &self:Spawn_NDCDynamicDestruction(Data Channel:NiagaraDataInterfaceDataChannelRead=$System.NDCReader, Max Count=300, Min Count=150, Scale Spawn Count=false) {module=Spawn_NDCDynamicDestruction}
  ]
  particleSpawnScript [
   &self:Initialize_NDCDynamicDestruction_Simple() {module=Initialize_NDCDynamicDestruction_Simple}
   &InitializeParticle(
    Color=<rgba 30 13.125 3 1>
    Lifetime Max=1
    Lifetime Min=0.15
    Lifetime Mode=::NewEnumerator1
    Position Mode=::NewEnumerator0
    Sprite Size Max=<3.5 2>
    Sprite Size Min=<2.5 1>
    Sprite Size Mode=::NewEnumerator2
    Write Material Random=true
   )
   &ShapeLocation(
    Additional Quaternion Rotation=false
    Box Size=&Multiply_VectorByFloat(Vector=$Particles.BoundsExtents, Float=2)
    MaskSpawnGroup=false
    Override Seed=false
    Rotation Mode=::NewEnumerator3
    Rotation Quaternion=$Particles.Orientation
    Shape Primitive=::NewEnumerator2
    Surface Only=false
    UseOwnerScale=true
   ) {module=ShapeLocation001}
  ]
  particleUpdateScript [
   &ParticleState()
   &KillParticles(Kill Particles=&Boolean_NotOperation(Bool=$Particles.bisCurie))
   &GravityForce(Gravity=<0 0 -600>)
   &CurlNoiseForce(
    Noise Frequency=300
    Noise Strength=&Multiply_Float(
     A=$Particles.MaterialRandom
     B=&FloatFromCurve(
      FloatCurve=DataInterface() {propertyValues {Curve {keys [
          <at 0 4147.9209 RCIM_Cubic>
          {arriveTangent=998.7806, interpMode=RCIM_Cubic, leaveTangent=998.7806, time=0.897, value=2308.9619}
          {arriveTangent=839.8494, interpMode=RCIM_Cubic, leaveTangent=839.8494, time=2.2666, value=6411.7881}
          {arriveTangent=818.7506, interpMode=RCIM_Cubic, leaveTangent=818.7506, time=3.6727, value=4640.1284}
          {arriveTangent=-197.3881, interpMode=RCIM_Cubic, leaveTangent=-197.3881, time=4.7165, value=8417.627}
          {arriveTangent=-812.5138, interpMode=RCIM_Cubic, leaveTangent=-812.5138, time=6.6551, value=4051.4272}
          {arriveTangent=379.0706, interpMode=RCIM_Cubic, leaveTangent=379.0706, time=9.3392, value=4661.6182}
          {arriveTangent=-88.1153, interpMode=RCIM_Cubic, leaveTangent=-88.1153, time=11.8318, value=6013.7651}
          <at 15.169 4147.9209 RCIM_Cubic>
         ]}}}
      CurveIndex=&Add_Float(A=$Engine.System.Age, B=$Particles.Age)
      Scale Curve=0.7
     )
    )
    Pan Noise=true
    Pan Noise Field=<0 0 1000>
   )
   &Drag(Drag=0.8, Use Linear Drag=true, Use Rotational Drag=false)
   &SolveForcesAndVelocity(Clamp Velocity=false, Limit Acceleration=false)
   &ScaleSpriteSizeBySpeed(Max Scale Factor=<1 10>, Velocity Threshold=1300)
   &ScaleColor(
    Scale Alpha=&FloatFromCurve(FloatCurve=DataInterface() {propertyValues {Curve {keys [<at 0 0>,<at 0.1997 0.9273>,<at 1.0011 0.002>]}}})
    ScaleA=true
    ScaleRGB=true
   )
  ]
  rendererClasses [&NiagaraSpriteRendererProperties]
  renderers [{rendererClass=&NiagaraSpriteRendererProperties}]
 }
 Dust_Smoke {
  emitterUpdateScript [
   &EmitterState()
   &self:Spawn_NDCDynamicDestruction(Data Channel:NiagaraDataInterfaceDataChannelRead=$System.NDCReader, Min Count=0) {module=Spawn_NDCDynamicDestruction}
  ]
  particleSpawnScript [
   &self:Initialize_NDCDynamicDestruction_Simple() {module=Initialize_NDCDynamicDestruction_Simple}
   /SetVariables(Particles.VisibilityTag=&SelectIntFromArray(Direct Array Index=$Particles.Resource, Int Selection Array=DataInterface() {propertyValues {IntData [0,0,0,0,0,1]}})) {isSetParameters=true}
   &InitializeParticle(
    Color Maximum=&MergeHelper_ColorSwitch(
     0=<rgba 0.2 0.1843 0.1686 0.25>
     1=<rgba 0.5 0.4608 0.4216 0.25>
     2=<rgba 0.2 0.1843 0.1686 0.25>
     3=<rgba 0.1 0.0922 0.0843 0.25>
     4=<rgba 1 0 0 1>
     5=$Particles.GlassColor
     Index=$Particles.Resource
     Num=::NewEnumerator4
    )
    Color Minimum=&MergeHelper_ColorSwitch(
     0=<rgba 0.1255 0.1137 0.1059 0.125>
     1=<rgba 0.333 0.3069 0.2808 0.125>
     2=<rgba 0.1255 0.1137 0.1059 0.125>
     3=<rgba 0.05 0.0453 0.0422 0.125>
     4=<rgba 1 0 0 1>
     5=$Particles.GlassColor
     Index=$Particles.Resource
     Num=::NewEnumerator4
    )
    Color Mode=::NewEnumerator2
    Lifetime Max=1
    Lifetime Min=0.75
    Lifetime Mode=::NewEnumerator1
    Position=$StackContext.HitLocation
    Position Mode=::NewEnumerator1
    Sprite Rotation Mode=::NewEnumerator2
    Sprite Size Mode=::NewEnumerator1
    Uniform Sprite Size Max=&MergeHelper_FloatSwitch(0=500, 1=500, 2=500, 3=500, 4=0, 5=300, Index=$Particles.Resource, Num=::NewEnumerator4)
    Uniform Sprite Size Min=&MergeHelper_FloatSwitch(0=375, 1=375, 2=375, 3=375, 4=0, 5=100, Index=$Particles.Resource, Num=::NewEnumerator4)
    UsePositionOffset=false
    Write Material Random=true
   )
   &CameraOffset(Camera Offset Amount=&RandomRangeFloat(
     Minimum=&MergeHelper_FloatSwitch(0=40, 1=40, 2=40, 3=40, 4=0, 5=0, Index=$Particles.Resource, Num=::NewEnumerator4)
     Maximum=&MergeHelper_FloatSwitch(0=50, 1=50, 2=50, 3=50, 4=0, 5=0, Index=$Particles.Resource, Num=::NewEnumerator4)
    ))
   /SetVariables(Particles.ForceApplied:Vector3f=&Multiply_VectorByFloat(
     Vector=&Multiply_Vector(A=$Particles.HitVector, B=<1 1 0.5>)
     Float=&Multiply_Float(
      A=&RandomRangeFloat(Minimum=750, Maximum=1500)
      B=&FloatFromCurve(
       FloatCurve=DataInterface() {propertyValues {Curve {keys [
           {arriveTangent=2.9421, interpMode=RCIM_Cubic, leaveTangent=2.9421, leaveTangentWeight=1.2157, tangentMode=RCTM_User, time=1, value=1}
           {arriveTangentWeight=0.6667, interpMode=RCIM_Cubic, tangentMode=RCTM_User, time=3, value=3.125}
          ]}}}
       CurveIndex=&ClampFloat(Float=&Divide_Float(A=$Particles.DamageDealt, B=50), Min=0.66, Max=3)
      )
     )
    )) {isSetParameters=true}
   &LinearForce(
    Coordinate Space:ENiagaraCoordinateSpace=::Simulation
    Force=&MergeHelper_VectorSwitch(0=<0 0 0>, 1=<0 0 0>, 2=<0 0 0>, 3=<0 0 0>, 4=<0 0 0>, 5=$Particles.ForceApplied, Index=$Particles.Resource, Num=::NewEnumerator4)
    Write to Intrinsic Parameters=false
   )
   &ApplyInitialForces()
   /SetVariables(Particles.Glass Dynamic Parameter:Vector4f=&MakeVector4(X=0.5, Y=0, Z=1, W=0.1)) {isSetParameters=true}
  ]
  particleUpdateScript [
   &ParticleState()
   /SetVariables(
    Particles.ScaleAlpha_Common:NiagaraFloat=&FloatFromCurve(FloatCurve=DataInterface() {propertyValues {Curve {keys [{interpMode=RCIM_Cubic, leaveTangentWeight=1.07, tangentMode=RCTM_User, time=0, value=1}, <at 1 0>]}}})
    Particles.ScaleColor_Common:NiagaraFloat=&FloatFromCurve(FloatCurve=DataInterface() {propertyValues {Curve {keys [{arriveTangent=1.8145, interpMode=RCIM_Cubic, leaveTangent=4.5795, tangentMode=RCTM_Break, time=0, value=0.4}, <at 0.3387 0.9565>, <at 1 1>]}}})
    Particles.ScaleSpriteSize_Common:NiagaraFloat=&FloatFromCurve(FloatCurve=DataInterface() {propertyValues {Curve {keys [
         {arriveTangent=8.1876, interpMode=RCIM_Cubic, leaveTangent=5.3162, tangentMode=RCTM_User, time=0, value=0.15}
         {arriveTangent=1.0025, interpMode=RCIM_Cubic, leaveTangent=1.0025, tangentMode=RCTM_User, time=0.161, value=0.7059}
         {arriveTangent=0.1638, interpMode=RCIM_Cubic, leaveTangent=0.1638, tangentMode=RCTM_User, time=1.0463, value=1}
        ]}}})
    Particles.SpriteRotationRate_Common:NiagaraFloat=&Multiply_Float(
     A=&RandomRangeFloat(Minimum=-2, Maximum=2)
     B=&FloatFromCurve(
      FloatCurve=DataInterface() {propertyValues {Curve {keys [
          {arriveTangent=-4.0107, interpMode=RCIM_Cubic, leaveTangent=-4.0107, tangentMode=RCTM_User, time=-0.0046, value=1.0113}
          {arriveTangent=-0.2496, interpMode=RCIM_Cubic, leaveTangent=-0.2496, tangentMode=RCTM_User, time=0.5, value=0.16}
         ]}}}
      Scale Curve=150
     )
    )
   ) {isSetParameters=true}
   &SpriteRotationRate(Rotation Rate=&MergeHelper_FloatSwitch(
     0=$Particles.SpriteRotationRate_Common
     1=$Particles.SpriteRotationRate_Common
     2=$Particles.SpriteRotationRate_Common
     3=$Particles.SpriteRotationRate_Common
     4=0
     5=0
     Index=$Particles.Resource
     Num=::NewEnumerator4
    ))
   &SubUVAnimation(
    Output Alternate Renderer Binding=true
    Playback Mode:ENiagara_FPSPlayrate=::NewEnumerator1
    "Sprite Renderer "=DataInterface() {propertyValues {SpriteRenderer=&self:Dust_Smoke_0.NiagaraSpriteRendererProperties_3}}
    SubUV Animation Mode:ENiagara_SubUVLookupModeV2=::NewEnumerator3
    UseEndFrame=false
    UsePlayRateScale=false
    UseStartFrame=false
   )
   &ScaleColor(
    Scale Alpha=&MergeHelper_FloatSwitch(
     0=$Particles.ScaleAlpha_Common
     1=$Particles.ScaleAlpha_Common
     2=$Particles.ScaleAlpha_Common
     3=$Particles.ScaleAlpha_Common
     4=0
     5=&FloatFromCurve(FloatCurve=DataInterface() {propertyValues {Curve {keys [
          {arriveTangent=-2.7426, interpMode=RCIM_Cubic, leaveTangent=-2.7426, tangentMode=RCTM_User, time=-0.0046, value=1.0113}
          {arriveTangent=-0.6319, interpMode=RCIM_Cubic, leaveTangent=-0.6319, tangentMode=RCTM_User, time=0.1699, value=0.2251}
          <at 1.0011 0.002 RCIM_Cubic>
         ]}}})
     Index=$Particles.Resource
     Num=::NewEnumerator4
    )
    Scale RGB=&VectorFromFloat(Value=&MergeHelper_FloatSwitch(
      0=$Particles.ScaleColor_Common
      1=$Particles.ScaleColor_Common
      2=$Particles.ScaleColor_Common
      3=$Particles.ScaleColor_Common
      4=0
      5=&FloatFromCurve(FloatCurve=DataInterface() {propertyValues {Curve {keys [
           <at 0 0 RCIM_Cubic>
           <at 0.2 1 RCIM_Cubic>
           {arriveTangent=-3.8119, interpMode=RCIM_Cubic, leaveTangent=-3.8119, tangentMode=RCTM_User, time=1, value=0}
          ]}}})
      Index=$Particles.Resource
      Num=::NewEnumerator4
     ))
    ScaleA=true
    ScaleRGB=true
   )
   &ScaleSpriteSize(
    Scale Sprite Size Mode=::NewEnumerator0
    Uniform Scale Factor=&MergeHelper_FloatSwitch(
     0=$Particles.ScaleSpriteSize_Common
     1=$Particles.ScaleSpriteSize_Common
     2=$Particles.ScaleSpriteSize_Common
     3=$Particles.ScaleSpriteSize_Common
     4=0
     5=&FloatFromCurve(FloatCurve=DataInterface() {propertyValues {Curve {keys [
          {arriveTangent=7.1197, interpMode=RCIM_Cubic, leaveTangent=7.1197, tangentMode=RCTM_User, time=0.0005, value=0.4065}
          {arriveTangent=1.6726, interpMode=RCIM_Cubic, leaveTangent=1.6726, tangentMode=RCTM_User, time=0.1945, value=1.384}
          <at 1.5 2 RCIM_Cubic>
         ]}}})
     Index=$Particles.Resource
     Num=::NewEnumerator4
    )
   )
   &DynamicMaterialParameters(
    Index 0 Param 1=&Add_Float(
     A=$Particles.MaterialRandom
     B=&FloatFromCurve(FloatCurve=DataInterface() {propertyValues {Curve {keys [
          {arriveTangent=2.7533, interpMode=RCIM_Cubic, leaveTangent=2.847, tangentMode=RCTM_Break, time=0, value=0}
          {arriveTangent=1.3581, interpMode=RCIM_Cubic, leaveTangent=1.3581, tangentMode=RCTM_User, time=0.2558, value=0.5125}
          {arriveTangent=0.0529, interpMode=RCIM_Cubic, leaveTangent=0.2021, tangentMode=RCTM_User, time=1, value=1}
         ]}}})
    )
    Index 0 Param 4=&FloatFromCurve(FloatCurve=DataInterface() {propertyValues {Curve {keys [<at 0 0.25>,<at 1 0>]}}})
    Index 2 Param 1=&FloatFromCurve(FloatCurve=DataInterface() {propertyValues {Curve {keys [
         {arriveTangent=2.646, interpMode=RCIM_Cubic, leaveTangent=2.646, tangentMode=RCTM_User, time=0, value=0}
         {leaveTangent=-0.0719, time=1, value=1}
        ]}}})
    Index 2 Param 2=&FloatFromCurve(
     FloatCurve=DataInterface() {propertyValues {Curve {keys [
         {arriveTangent=4.4413, interpMode=RCIM_Cubic, leaveTangent=4.4413, leaveTangentWeight=0.3333, tangentMode=RCTM_User, time=0, value=0.2}
         {arriveTangentWeight=0.3333, interpMode=RCIM_Cubic, tangentMode=RCTM_User, time=1, value=1.4}
        ]}}}
     Scale Curve=&RandomRangeFloat(Minimum=0.5, Maximum=1)
    )
    Index 2 Param 3=&FloatFromCurve(FloatCurve=DataInterface() {propertyValues {Curve {keys [
         {interpMode=RCIM_Cubic, leaveTangent=3.2672, tangentMode=RCTM_Break, time=0, value=1}
         {interpMode=RCIM_Cubic, leaveTangent=11.1233, tangentMode=RCTM_Break, time=1, value=2}
        ]}}})
    Index 2 Param 4=&FloatFromCurve(
     FloatCurve=DataInterface() {propertyValues {Curve {keys [
         {arriveTangent=-0.6867, interpMode=RCIM_Cubic, leaveTangent=-0.6867, tangentMode=RCTM_User, time=0, value=1}
         {interpMode=RCIM_Cubic, leaveTangent=-1.2956, tangentMode=RCTM_Break, time=1, value=0.75}
        ]}}}
     Scale Curve=1.6
    )
    Module.Index0Float0Write=true
    Module.Index0Float1Write=false
    Module.Index0Float2Write=true
    Module.Index0Float3Write=true
    Module.Index2Float0Write=true
    Module.Index2Float1Write=true
    Module.Index2Float2Write=true
    Module.Index2Float3Write=true
    Write Parameter Index 2=::NewEnumerator1
   )
   &AccelerationForce(Acceleration=&MergeHelper_VectorSwitch(0=<0 0 250>, 1=<0 0 250>, 2=<0 0 250>, 3=<0 0 250>, 4=<0 0 250>, 5=<0 0 -600>, Index=$Particles.Resource, Num=::NewEnumerator4))
   &Drag(Drag=&RandomRangeFloat(Minimum=1, Maximum=3), Use Linear Drag=true, Use Rotational Drag=false)
   &SolveForcesAndVelocity(Clamp Velocity=false, Limit Acceleration=false)
  ]
  rendererClasses [&NiagaraSpriteRendererProperties]
  renderers [
   {rendererClass=&NiagaraSpriteRendererProperties}
   {rendererClass=&NiagaraSpriteRendererProperties, rendererIndex=1}
   {rendererClass=&NiagaraSpriteRendererProperties, rendererIndex=2}
  ]
 }
 Metal_Sparks {
  simTarget=GPUComputeSim
  emitterUpdateScript [
   &EmitterState()
   &self:Spawn_NDCDynamicDestruction(Data Channel:NiagaraDataInterfaceDataChannelRead=$System.NDCReader, Max Count=300, Min Count=120, Scale Spawn Count=false) {module=Spawn_NDCDynamicDestruction}
  ]
  particleSpawnScript [
   &self:Initialize_NDCDynamicDestruction_Simple() {module=Initialize_NDCDynamicDestruction_Simple}
   &InitializeParticle(
    Color=<rgba 30 26.9412 16 1>
    Lifetime Max=0.45
    Lifetime Min=0.1
    Lifetime Mode=::NewEnumerator1
    Position Mode=::NewEnumerator0
    Sprite Size Max=<2 2>
    Sprite Size Min=<1 1>
    Sprite Size Mode=::NewEnumerator2
    Write Material Random=true
   )
   &ShapeLocation(
    Additional Quaternion Rotation=false
    Box Size=&Multiply_VectorByFloat(Vector=$Particles.BoundsExtents, Float=2)
    MaskSpawnGroup=false
    Override Seed=false
    Rotation Mode=::NewEnumerator3
    Rotation Quaternion=$Particles.Orientation
    Shape Primitive=::NewEnumerator2
    Surface Only=false
    UseOwnerScale=true
   ) {module=ShapeLocation001}
   &self:Place_On_DF_Surface(KillOnFailure=true, Max Distance=2000, StaticMeshID:NiagaraSharedDataID_StaticMesh=$Particles.StaticMeshID, Use Max Distance=true) {module=Place_On_DF_Surface}
   &ShapeLocation(MaskSpawnGroup=false, Override Seed=false, Sphere Radius=5, UseOwnerScale=true)
   &AddVelocity(Velocity=&Multiply_VectorByFloat(Vector=$Particles.ShapeLocation.ShapeNormal, Float=&RandomRangeFloat(Minimum=350, Maximum=750)))
   &CameraOffset(Camera Offset Amount=25)
  ]
  particleUpdateScript [
   &ParticleState()
   &KillParticles(Kill Particles=&SetBoolByIntComparison(A=$Particles.Resource, B=3, Comparison Type=::NewEnumerator5))
   &GravityForce()
   &Drag(Drag=&RandomRangeFloat(Minimum=3, Maximum=10), Use Linear Drag=true, Use Rotational Drag=false)
   &SolveForcesAndVelocity(Clamp Velocity=false, Limit Acceleration=false)
   &ScaleSpriteSizeBySpeed(Max Scale Factor=<1 10>)
   &ScaleSpriteSize(Uniform Curve Sprite Scale=DataInterface() {propertyValues {Curve {keys [{arriveTangent=-6.725, interpMode=RCIM_Cubic, leaveTangent=-6.725, tangentMode=RCTM_User, time=0.65, value=1}, <at 1 0 RCIM_Cubic>]}}})
   &ScaleColor(
    Scale RGB=&VectorFromCurve(VectorCurve=DataInterface() {propertyValues {
       XCurve {keys [<at 0.5 12 RCIM_Cubic>,<at 1 1 RCIM_Cubic>]}
       YCurve {keys [<at 0.5 6 RCIM_Cubic>,<at 1 1 RCIM_Cubic>]}
       ZCurve {keys [<at 0.5 2 RCIM_Cubic>,<at 1 1 RCIM_Cubic>]}
      }})
    ScaleA=true
    ScaleRGB=true
   )
  ]
  rendererClasses [&NiagaraSpriteRendererProperties]
  renderers [{rendererClass=&NiagaraSpriteRendererProperties}]
 }
 Metal_Sparks_Mobile {
  emitterUpdateScript [
   &EmitterState()
   &self:Spawn_NDCDynamicDestruction(Data Channel:NiagaraDataInterfaceDataChannelRead=$System.NDCReader, Max Count=150, Min Count=60, Scale Spawn Count=false) {module=Spawn_NDCDynamicDestruction}
  ]
  particleSpawnScript [
   &self:Initialize_NDCDynamicDestruction_Simple() {module=Initialize_NDCDynamicDestruction_Simple}
   &InitializeParticle(
    Color=<rgba 30 26.9412 16 1>
    Lifetime Max=0.45
    Lifetime Min=0.1
    Lifetime Mode=::NewEnumerator1
    Position Mode=::NewEnumerator0
    Sprite Size Max=<2 2>
    Sprite Size Min=<1 1>
    Sprite Size Mode=::NewEnumerator2
    Write Material Random=true
   )
   &ShapeLocation(
    Additional Quaternion Rotation=false
    Box Size=&Multiply_VectorByFloat(Vector=$Particles.BoundsExtents, Float=2)
    MaskSpawnGroup=false
    Override Seed=false
    Rotation Mode=::NewEnumerator3
    Rotation Quaternion=$Particles.Orientation
    Shape Primitive=::NewEnumerator2
    Surface Only=false
    UseOwnerScale=true
   ) {module=ShapeLocation001}
   &ShapeLocation(MaskSpawnGroup=false, Override Seed=false, Sphere Radius=5, UseOwnerScale=true)
   &AddVelocity(Velocity=&Multiply_VectorByFloat(Vector=$Particles.ShapeLocation.ShapeNormal, Float=&RandomRangeFloat(Minimum=350, Maximum=750)))
   &CameraOffset(Camera Offset Amount=25)
  ]
  particleUpdateScript [
   &ParticleState()
   &KillParticles(Kill Particles=&SetBoolByIntComparison(A=$Particles.Resource, B=3, Comparison Type=::NewEnumerator5))
   &GravityForce()
   &Drag(Drag=&RandomRangeFloat(Minimum=3, Maximum=10), Use Linear Drag=true, Use Rotational Drag=false)
   &SolveForcesAndVelocity(Clamp Velocity=false, Limit Acceleration=false)
   &ScaleSpriteSizeBySpeed(Max Scale Factor=<1 10>)
   &ScaleSpriteSize(Uniform Curve Sprite Scale=DataInterface() {propertyValues {Curve {keys [{arriveTangent=-6.725, interpMode=RCIM_Cubic, leaveTangent=-6.725, tangentMode=RCTM_User, time=0.65, value=1}, <at 1 0 RCIM_Cubic>]}}})
   &ScaleColor(
    Scale RGB=&VectorFromCurve(VectorCurve=DataInterface() {propertyValues {
       XCurve {keys [<at 0.5 12 RCIM_Cubic>,<at 1 1 RCIM_Cubic>]}
       YCurve {keys [<at 0.5 6 RCIM_Cubic>,<at 1 1 RCIM_Cubic>]}
       ZCurve {keys [<at 0.5 2 RCIM_Cubic>,<at 1 1 RCIM_Cubic>]}
      }})
    ScaleA=true
    ScaleRGB=true
   )
  ]
  rendererClasses [&NiagaraSpriteRendererProperties]
  renderers [{rendererClass=&NiagaraSpriteRendererProperties}]
 }
 Resource {
  simTarget=GPUComputeSim
  emitterUpdateScript [
   &EmitterState()
   &self:Spawn_NDCDynamicDestruction(Data Channel:NiagaraDataInterfaceDataChannelRead=$System.NDCReader, Max Count=120, Min Count=60) {module=Spawn_NDCDynamicDestruction}
  ]
  particleSpawnScript [
   &self:Initialize_NDCDynamicDestruction_Resources(
    Default LIfetime Max=2
    Default Lifetime Min=1
    Default Mass 0:NiagaraFloat=&RandomRangeFloat(Minimum=2, Maximum=7)
    Default Mass 1:NiagaraFloat=&RandomRangeFloat(Minimum=3, Maximum=9)
    Default Mass 2:NiagaraFloat=&RandomRangeFloat(Minimum=5, Maximum=17)
    Default Scale 0:Vector3f=&VectorFromFloat(Value=&RandomRangeFloat(Minimum=0.5, Maximum=1))
    Default Scale 1:Vector3f=&VectorFromFloat(Value=&RandomRangeFloat(Minimum=1, Maximum=3))
    Default Scale 2:Vector3f=&VectorFromFloat(Value=&RandomRangeFloat(Minimum=4, Maximum=6))
    Foliage Lifetime Max=4
    Foliage Lifetime Min=2
    Foliage Mass:NiagaraFloat=&RandomRangeFloat(Minimum=0.5, Maximum=1.5)
    Foliage Scale:Vector3f=&VectorFromFloat(Value=&RandomRangeFloat(Minimum=5, Maximum=12))
    Glass Lifetime Max=2
    Glass Lifetime Min=1
    Glass Mass 0:NiagaraFloat=&RandomRangeFloat(Minimum=1, Maximum=4)
    Glass Mass 1:NiagaraFloat=&RandomRangeFloat(Minimum=1, Maximum=4)
    Glass Mass 2:NiagaraFloat=&RandomRangeFloat(Minimum=1, Maximum=4)
    Glass Scale 0:Vector3f=&VectorFromFloat(Value=&RandomRangeFloat(Minimum=1, Maximum=2))
    Glass Scale 1:Vector3f=&VectorFromFloat(Value=&RandomRangeFloat(Minimum=1, Maximum=2))
    Glass Scale 2:Vector3f=&VectorFromFloat(Value=&RandomRangeFloat(Minimum=2, Maximum=3))
    ImpulseScaleCurve=DataInterface() {propertyValues {Curve {keys [
        {arriveTangent=2.9421, interpMode=RCIM_Cubic, leaveTangent=2.9421, leaveTangentWeight=1.2157, tangentMode=RCTM_User, time=1, value=1}
        {arriveTangentWeight=0.6667, interpMode=RCIM_Cubic, tangentMode=RCTM_User, time=3, value=3.125}
       ]}}}
    Medium Mesh=0.96
    Metal Lifetime Max=2
    Metal Lifetime Min=1
    Metal Mass 0:NiagaraFloat=&RandomRangeFloat(Minimum=2.5, Maximum=7)
    Metal Mass 1:NiagaraFloat=&RandomRangeFloat(Minimum=3.5, Maximum=9)
    Metal Mass 2:NiagaraFloat=&RandomRangeFloat(Minimum=5.5, Maximum=17)
    Metal Scale 0:Vector3f=&VectorFromFloat(Value=&RandomRangeFloat(Minimum=0.5, Maximum=1))
    Metal Scale 1:Vector3f=&VectorFromFloat(Value=&RandomRangeFloat(Minimum=0.15, Maximum=0.6))
    Metal Scale 2:Vector3f=&VectorFromFloat(Value=&RandomRangeFloat(Minimum=0.3, Maximum=1))
    Module.Override Randomness=false
    Small Mesh=0.5
    Stone Lifetime Max=2
    Stone Lifetime Min=1
    Stone Mass 0:NiagaraFloat=&RandomRangeFloat(Minimum=2, Maximum=7)
    Stone Mass 1:NiagaraFloat=&RandomRangeFloat(Minimum=3, Maximum=9)
    Stone Mass 2:NiagaraFloat=&RandomRangeFloat(Minimum=5, Maximum=17)
    Stone Scale 0:Vector3f=&VectorFromFloat(Value=&RandomRangeFloat(Minimum=0.3, Maximum=1))
    Stone Scale 1:Vector3f=&VectorFromFloat(Value=&RandomRangeFloat(Minimum=0.5, Maximum=2))
    Stone Scale 2:Vector3f=&VectorFromFloat(Value=&RandomRangeFloat(Minimum=4, Maximum=6))
    Wood Lifetime Max=2
    Wood Lifetime Min=1
    Wood Mass 0:NiagaraFloat=&RandomRangeFloat(Minimum=1.5, Maximum=8)
    Wood Mass 1:NiagaraFloat=&RandomRangeFloat(Minimum=2.5, Maximum=9)
    Wood Mass 2:NiagaraFloat=&RandomRangeFloat(Minimum=4.5, Maximum=17)
    Wood Scale 0:Vector3f=&VectorFromFloat(Value=&RandomRangeFloat(Minimum=0.8, Maximum=1.2))
    Wood Scale 1:Vector3f=&VectorFromFloat(Value=&RandomRangeFloat(Minimum=2, Maximum=3))
    Wood Scale 2:Vector3f=&VectorFromFloat(Value=&RandomRangeFloat(Minimum=0.75, Maximum=1))
   ) {module=Initialize_NDCDynamicDestruction_Resources}
   &InitializeParticle(
    Lifetime Max=$Particles.Resource Lifetime Max
    Lifetime Min:NiagaraFloat=$Particles.Resource LIfetime Min
    Lifetime Mode=::NewEnumerator1
    Mass=$Particles.ResourceMass
    Mass Mode=::NewEnumerator1
    Mesh Scale=&self:Mesh_Dynamic_Bounds_Scaling(
     Resource Scale:Vector3f=$Particles.ResourceScale
     Mesh Scale:Vector3f=<1 1 1>
     Sphere Radius:NiagaraFloat=$Particles.BoundsRadius
     Comparison Value:NiagaraFloat=384
    )
    Mesh Scale Mode=::NewEnumerator4
    Position Mode=::NewEnumerator0
    Sprite Size Mode=::NewEnumerator3
    Uniform Sprite Size=&Multiply_Float(A=1, B=&MakeFloatFromVector(Vector=$Particles.ResourceScale, Channel=::NewEnumerator0))
    Write Material Random=true
   )
   &ShapeLocation(
    Additional Quaternion Rotation=false
    Box Size=&Multiply_VectorByFloat(Vector=$Particles.BoundsExtents, Float=2)
    MaskSpawnGroup=false
    Override Seed=false
    Rotation Mode=::NewEnumerator3
    Rotation Quaternion=$Particles.Orientation
    Shape Primitive=::NewEnumerator2
    Surface Only=false
    UseOwnerScale=true
   )
   &self:Place_On_DF_Surface(KillOnFailure=true, Max Distance=2000, StaticMeshID:NiagaraSharedDataID_StaticMesh=$Particles.StaticMeshID, Use Max Distance=true) {module=Place_On_DF_Surface}
   &self:Pull_Gbuffer_Color(
    Fallback Default Color:LinearColor=<rgba 0 0 0 1>
    Fallback Foliage Color:LinearColor=<rgba 0.1533 0.3403 0.0284 1>
    Fallback Glass Color:LinearColor=$Particles.GlassColor
    Fallback Metal Color:LinearColor=<rgba 0.07 0.07 0.07 1>
    Fallback Stone Color:LinearColor=<rgba 0.1356 0.1356 0.1343 1>
    Fallback Wood Color:LinearColor=<rgba 0.1354 0.0892 0.0432 1>
    Resource:NiagaraInt32=$Particles.Resource
   ) {module=Pull_Gbuffer_Color}
   &InitialMeshOrientation(Mesh Orientation Mode=::NewEnumerator2, Orientation Vector=$Output.Place_On_DF_Surface.OutClosestNormal, Use Rotation Vector=false)
   &PointForce(
    Force Falloff Distance:NiagaraFloat=$Particles.Radius
    Force Origin:NiagaraPosition=$Particles.HitLocation
    Force Strength=&MakeCustomFloatFromBool(
     Bool=$Particles.bIsInstantDeath
     True Float=&Multiply_Float(A=&RandomRangeFloat(Minimum=1000, Maximum=1500), B=$Particles.Impulse Scale)
     False Float=&Multiply_Float(A=&RandomRangeFloat(Minimum=1750, Maximum=2250), B=$Particles.Impulse Scale)
    )
    Module.Use Falloff=true
    Module.UseLinearForce=false
   )
   &LinearForce(
    Coordinate Space:ENiagaraCoordinateSpace=::World
    Force=&Multiply_VectorByFloat(
     Vector=&Multiply_Vector(A=$Particles.HitVector, B=<1 1 0>)
     Float=&MakeCustomFloatFromBool(
      Bool=$Particles.bIsInstantDeath
      True Float=&Multiply_Float(A=&RandomRangeFloat(Minimum=1250, Maximum=1750), B=$Particles.Impulse Scale)
      False Float=&Multiply_Float(A=&RandomRangeFloat(Minimum=1500, Maximum=2000), B=$Particles.Impulse Scale)
     )
    )
   )
   &ApplyInitialForces()
  ]
  particleUpdateScript [
   &ParticleState()
   &GravityForce(Gravity=&Multiply_VectorByFloat(
     Vector=<0 0 -980>
     Float=&FloatFromCurve(FloatCurve=DataInterface() {propertyValues {Curve {
         defaultValue=1
         keys [<at 0 0 RCIM_Cubic>, {arriveTangent=30.1745, interpMode=RCIM_Cubic, leaveTangent=30.1745, tangentMode=RCTM_User, time=0.1, value=1}]
        }}})
    ))
   &Drag(
    Drag=&self:Gravity_Drag_Compensation(Drag Amount:NiagaraFloat=1, Drag Falloff:NiagaraFloat=2, Gravity Direction:Vector3f=<0 0 -1>)
    Use Linear Drag=true
    Use Rotational Drag=true
   )
   &Collision(
    Age Colliding Particles=false
    Collision Enabled=&RandomBool()
    Friction During a Bounce=0.2
    GPU Collision Type=::NewEnumerator2
    KillOnCollision=false
    Mesh Dimensions:Vector3f=&self:Spawned_Mesh_Dimensions(
     Default Scale 0:Vector3f=<18 16 21>
     Default Scale 1:Vector3f=<43 39 42>
     Default Scale 2:Vector3f=<47 60 46>
     Wood Scale 0:Vector3f=<65 29 34>
     Wood Scale 1:Vector3f=<96 21 3>
     Wood Scale 2:Vector3f=<69 17 18>
     Stone Scale 0:Vector3f=<18 16 21>
     Stone Scale 1:Vector3f=<43 39 42>
     Stone Scale 2:Vector3f=<47 60 46>
     Metal Scale 0:Vector3f=<32 32 6>
     Metal Scale 1:Vector3f=<52 52 13>
     Metal Scale 2:Vector3f=<16 7 9>
     Foliage Scale:Vector3f=<1 1 1>
     Glass Scale 0:Vector3f=<12 5 1>
     Glass Scale 1:Vector3f=<16 11 1>
     Glass Scale 2:Vector3f=<22 13 2>
    )
    Module.Control Rest State Via Surface Normal=false
    Module.EnableMaxCollisionCount=false
    Module.Randomize Collision Normal Vector=true
    Particle Radius Scale=0.1
    Radius Calculation Type=::NewEnumerator1
    Restitution=0.15
   )
   &SolveForcesAndVelocity(Clamp Velocity=false, Limit Acceleration=false)
   &UpdateMeshOrientation(Rotation Vector=&MakeCustomVectorFromBool(
     Bool=$Particles.HasCollided
     True Vector=<0 0 0>
     False Vector=&RandomRangeVector(Minimum=<-1 -1 -1>, Maximum=<1 1 1>)
    ))
   &ScaleMeshSize(Scale Factor=&VectorFromFloat(Value=&FloatFromCurve(FloatCurve=DataInterface() {propertyValues {Curve {keys [
          <at 0 0 RCIM_Constant>
          <at 0.05 1 RCIM_Constant>
          {interpMode=RCIM_Cubic, leaveTangentWeight=0.3333, tangentMode=RCTM_User, time=0.75, value=1}
          {arriveTangentWeight=0.3333, interpMode=RCIM_Cubic, tangentMode=RCTM_User, time=1, value=0}
         ]}}})))
  ]
  rendererClasses [&NiagaraMeshRendererProperties]
  renderers [
   {rendererClass=&NiagaraMeshRendererProperties, rendererIndex=0}
   {rendererClass=&NiagaraMeshRendererProperties, rendererIndex=1}
   {rendererClass=&NiagaraMeshRendererProperties, rendererIndex=2}
   {rendererClass=&NiagaraMeshRendererProperties, rendererIndex=3}
   {rendererClass=&NiagaraMeshRendererProperties, rendererIndex=4}
   {rendererClass=&NiagaraMeshRendererProperties, rendererIndex=5}
  ]
 }
 Resource_Mobile {
  emitterUpdateScript [
   &EmitterState()
   &self:Spawn_NDCDynamicDestruction(Data Channel:NiagaraDataInterfaceDataChannelRead=$System.NDCReader, Max Count=60, Min Count=30) {module=Spawn_NDCDynamicDestruction}
  ]
  particleSpawnScript [
   &self:Initialize_NDCDynamicDestruction_Resources(
    Default LIfetime Max=2
    Default Lifetime Min=1
    Default Mass 0:NiagaraFloat=&RandomRangeFloat(Minimum=2, Maximum=7)
    Default Mass 1:NiagaraFloat=&RandomRangeFloat(Minimum=3, Maximum=9)
    Default Mass 2:NiagaraFloat=&RandomRangeFloat(Minimum=5, Maximum=17)
    Default Scale 0:Vector3f=&VectorFromFloat(Value=&RandomRangeFloat(Minimum=0.5, Maximum=1.5))
    Default Scale 1:Vector3f=&VectorFromFloat(Value=&RandomRangeFloat(Minimum=0.75, Maximum=1.25))
    Default Scale 2:Vector3f=&VectorFromFloat(Value=&RandomRangeFloat(Minimum=1.25, Maximum=1.6))
    Foliage Lifetime Max=4
    Foliage Lifetime Min=2
    Foliage Mass:NiagaraFloat=&RandomRangeFloat(Minimum=1, Maximum=2.5)
    Foliage Scale:Vector3f=&VectorFromFloat(Value=&RandomRangeFloat(Minimum=0.5, Maximum=2))
    Glass Lifetime Max=2
    Glass Lifetime Min=1
    Glass Mass 0:NiagaraFloat=&RandomRangeFloat(Minimum=1, Maximum=4)
    Glass Mass 1:NiagaraFloat=&RandomRangeFloat(Minimum=1, Maximum=4)
    Glass Mass 2:NiagaraFloat=&RandomRangeFloat(Minimum=1, Maximum=4)
    Glass Scale 0:Vector3f=&VectorFromFloat(Value=&RandomRangeFloat(Minimum=1, Maximum=2))
    Glass Scale 1:Vector3f=&VectorFromFloat(Value=&RandomRangeFloat(Minimum=1, Maximum=2))
    Glass Scale 2:Vector3f=&VectorFromFloat(Value=&RandomRangeFloat(Minimum=2, Maximum=3))
    ImpulseScaleCurve=DataInterface() {propertyValues {Curve {keys [
        {arriveTangent=2.9421, interpMode=RCIM_Cubic, leaveTangent=2.9421, leaveTangentWeight=1.2157, tangentMode=RCTM_User, time=1, value=1}
        {arriveTangentWeight=0.6667, interpMode=RCIM_Cubic, tangentMode=RCTM_User, time=3, value=3.125}
       ]}}}
    Medium Mesh=0.96
    Metal Lifetime Max=2
    Metal Lifetime Min=1
    Metal Mass 0:NiagaraFloat=&RandomRangeFloat(Minimum=2.5, Maximum=7)
    Metal Mass 1:NiagaraFloat=&RandomRangeFloat(Minimum=3.5, Maximum=9)
    Metal Mass 2:NiagaraFloat=&RandomRangeFloat(Minimum=5.5, Maximum=17)
    Metal Scale 0:Vector3f=&VectorFromFloat(Value=&RandomRangeFloat(Minimum=0.5, Maximum=1.6))
    Metal Scale 1:Vector3f=&VectorFromFloat(Value=&RandomRangeFloat(Minimum=0.5, Maximum=1.2))
    Metal Scale 2:Vector3f=&VectorFromFloat(Value=&RandomRangeFloat(Minimum=0.8, Maximum=1.2))
    Module.Override Randomness=false
    Small Mesh=0.5
    Stone Lifetime Max=2
    Stone Lifetime Min=1
    Stone Mass 0:NiagaraFloat=&RandomRangeFloat(Minimum=2, Maximum=7)
    Stone Mass 1:NiagaraFloat=&RandomRangeFloat(Minimum=3, Maximum=9)
    Stone Mass 2:NiagaraFloat=&RandomRangeFloat(Minimum=5, Maximum=17)
    Stone Scale 0:Vector3f=&VectorFromFloat(Value=&RandomRangeFloat(Minimum=0.5, Maximum=1.5))
    Stone Scale 1:Vector3f=&VectorFromFloat(Value=&RandomRangeFloat(Minimum=0.75, Maximum=1.25))
    Stone Scale 2:Vector3f=&VectorFromFloat(Value=&RandomRangeFloat(Minimum=1.25, Maximum=1.6))
    Wood Lifetime Max=2
    Wood Lifetime Min=1
    Wood Mass 0:NiagaraFloat=&RandomRangeFloat(Minimum=1.5, Maximum=8)
    Wood Mass 1:NiagaraFloat=&RandomRangeFloat(Minimum=2.5, Maximum=9)
    Wood Mass 2:NiagaraFloat=&RandomRangeFloat(Minimum=4.5, Maximum=17)
    Wood Scale 0:Vector3f=&VectorFromFloat(Value=&RandomRangeFloat(Minimum=0.5, Maximum=1.3))
    Wood Scale 1:Vector3f=&VectorFromFloat(Value=&RandomRangeFloat(Minimum=0.5, Maximum=1.6))
    Wood Scale 2:Vector3f=&VectorFromFloat(Value=&RandomRangeFloat(Minimum=1, Maximum=1.6))
   ) {module=Initialize_NDCDynamicDestruction_Resources}
   &InitializeParticle(
    Color=&MergeHelper_ColorSwitch(
     0=<rgba 0 0 0 1>
     1=<rgba 0.1354 0.0892 0.0432 1>
     2=<rgba 0.1356 0.1356 0.1343 1>
     3=<rgba 0.07 0.07 0.07 1>
     4=<rgba 0.1533 0.3403 0.0284 1>
     5=$Particles.GlassColor
     Index=$Particles.Resource
     Num=::NewEnumerator4
    )
    Lifetime Max=$Particles.Resource Lifetime Max
    Lifetime Min:NiagaraFloat=$Particles.Resource LIfetime Min
    Lifetime Mode=::NewEnumerator1
    Mass=$Particles.ResourceMass
    Mass Mode=::NewEnumerator1
    Mesh Scale=&self:Mesh_Dynamic_Bounds_Scaling(
     Resource Scale:Vector3f=$Particles.ResourceScale
     Mesh Scale:Vector3f=<1 1 1>
     Sphere Radius:NiagaraFloat=0
     Comparison Value:NiagaraFloat=384
    )
    Mesh Scale Mode=::NewEnumerator4
    Sprite Size Mode=::NewEnumerator3
    Uniform Sprite Size=&Multiply_Float(A=1, B=&MakeFloatFromVector(Vector=$Particles.ResourceScale, Channel=::NewEnumerator0))
    UsePositionOffset=false
    Write Material Random=true
   )
   &ShapeLocation(
    Additional Quaternion Rotation=false
    Box Size=&Multiply_VectorByFloat(Vector=$Particles.BoundsExtents, Float=2)
    MaskSpawnGroup=false
    Override Seed=false
    Rotation Mode=::NewEnumerator3
    Rotation Quaternion=$Particles.Orientation
    Shape Primitive=::NewEnumerator2
    Surface Only=false
    UseOwnerScale=true
   )
   &self:Place_On_DF_Surface(KillOnFailure=true, Max Distance=2000, StaticMeshID:NiagaraSharedDataID_StaticMesh=$Particles.StaticMeshID, Use Max Distance=true) {enabled=false, module=Place_On_DF_Surface}
   &InitialMeshOrientation(Mesh Orientation Mode=::NewEnumerator2, Orientation Vector=$Particles.ShapeLocation.ShapeNormal, Use Rotation Vector=false)
   &PointForce(
    Force Falloff Distance:NiagaraFloat=$Particles.Radius
    Force Origin:NiagaraPosition=$Particles.HitLocation
    Force Strength=&MakeCustomFloatFromBool(
     Bool=true
     True Float=&Multiply_Float(A=&RandomRangeFloat(Minimum=1000, Maximum=1500), B=$Particles.Impulse Scale)
     False Float=&Multiply_Float(A=&RandomRangeFloat(Minimum=1750, Maximum=2250), B=$Particles.Impulse Scale)
    )
    Module.Use Falloff=true
    Module.UseLinearForce=false
   )
   &LinearForce(
    Coordinate Space:ENiagaraCoordinateSpace=::World
    Force=&Multiply_VectorByFloat(
     Vector=&Multiply_Vector(A=$Particles.HitVector, B=<1 1 0>)
     Float=&MakeCustomFloatFromBool(
      Bool=true
      True Float=&Multiply_Float(A=&RandomRangeFloat(Minimum=1250, Maximum=1750), B=$Particles.Impulse Scale)
      False Float=&Multiply_Float(A=&RandomRangeFloat(Minimum=1500, Maximum=2000), B=$Particles.Impulse Scale)
     )
    )
   )
   &ApplyInitialForces()
  ]
  particleUpdateScript [
   &ParticleState()
   &GravityForce(Gravity=&Multiply_VectorByFloat(
     Vector=<0 0 -980>
     Float=&FloatFromCurve(FloatCurve=DataInterface() {propertyValues {Curve {
         defaultValue=1
         keys [<at 0 0 RCIM_Cubic>, {arriveTangent=30.1745, interpMode=RCIM_Cubic, leaveTangent=30.1745, tangentMode=RCTM_User, time=0.1, value=1}]
        }}})
    ))
   &Drag(
    Drag=&self:Gravity_Drag_Compensation(Drag Amount:NiagaraFloat=1, Drag Falloff:NiagaraFloat=2, Gravity Direction:Vector3f=<0 0 -1>)
    Use Linear Drag=true
    Use Rotational Drag=true
   )
   &SolveForcesAndVelocity(Clamp Velocity=false, Limit Acceleration=false)
   &UpdateMeshOrientation(Rotation Vector=&RandomRangeVector(Minimum=<-1 -1 -1>, Maximum=<1 1 1>))
   &ScaleMeshSize(Scale Factor=&VectorFromFloat(Value=&FloatFromCurve(FloatCurve=DataInterface() {propertyValues {Curve {keys [
          <at 0 0 RCIM_Constant>
          <at 0.05 1 RCIM_Constant>
          {interpMode=RCIM_Cubic, leaveTangentWeight=0.3333, tangentMode=RCTM_User, time=0.75, value=1}
          {arriveTangentWeight=0.3333, interpMode=RCIM_Cubic, tangentMode=RCTM_User, time=1, value=0}
         ]}}})))
  ]
  rendererClasses [&NiagaraMeshRendererProperties]
  renderers [
   {rendererClass=&NiagaraMeshRendererProperties, rendererIndex=0}
   {rendererClass=&NiagaraMeshRendererProperties, rendererIndex=1}
   {rendererClass=&NiagaraMeshRendererProperties, rendererIndex=2}
   {rendererClass=&NiagaraMeshRendererProperties, rendererIndex=3}
   {rendererClass=&NiagaraMeshRendererProperties, rendererIndex=4}
   {rendererClass=&NiagaraMeshRendererProperties, rendererIndex=5}
  ]
 }
}
summary {
 systemName=NS_NDC_DynamicDestruction
 emitters [
  {emitterName=Resource, simTarget=GPUComputeSim, rendererClasses [&NiagaraMeshRendererProperties]}
  {emitterName=Dust_Smoke, rendererClasses [&NiagaraSpriteRendererProperties]}
  {emitterName=Resource_Mobile, rendererClasses [&NiagaraMeshRendererProperties]}
  {emitterName=Metal_Sparks, simTarget=GPUComputeSim, rendererClasses [&NiagaraSpriteRendererProperties]}
  {emitterName=Curie_Sparks, simTarget=GPUComputeSim, rendererClasses [&NiagaraSpriteRendererProperties]}
  {emitterName=Metal_Sparks_Mobile, rendererClasses [&NiagaraSpriteRendererProperties]}
  {emitterName=Curie_Sparks_Mobile, rendererClasses [&NiagaraSpriteRendererProperties]}
 ]
 userVariables [User.Bounds {type=Vector3f, description="", defaultValue {struct=Vector3f, value=<1 500 500>}}]
}
